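# Physics asset for the front-end mesh of the BlessFlan medium male player body, covering its dangling belts, buckle straps and front skirt.
# Capsules on pelvis and thighs form the body volume, while capsule and box chains follow the cloth and strap bones.
# Belt and buckle chains hang from tiny anchor bodies with tight swing and twist limits.
# Skirt chains, anchors and thighs link back to the pelvis through angularly free constraints.
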PhysicsAsset /BRCosmetics/Characters/Player/Male/Medium/Bodies/M_MED_BlessFlan/Meshes/M_MED_BlessFlan_FrontEnd_Physics
bodies {
 MSTR_Belt [{shapeType=Capsule, length=0.505, radius=0.505, shapeName=MSTR_Belt_capsule, rotation=<rot 0 -90 0>}]
 MSTR_frontSkirt_l [
  {
   shapeType=Capsule
   length=10.7501
   radius=3.9994
   shapeName=MSTR_frontSkirt_l_capsule
   center=<-0.2016 -0.7889 -3.3399>
   rotation=<rot 76.4024 89.9485 -179.9995>
  }
  {
   shapeType=Capsule
   length=12.9635
   radius=4.4067
   shapeName=MSTR_frontSkirt_capsule
   center=<-5.8704 0.0136 -4.7964>
   rotation=<rot 0.0002 0 -82.7332>
  }
 ]
 MSTR_frontSkirt_r [
  {
   shapeType=Capsule
   length=8
   radius=4
   shapeName=MSTR_frontSkirt_r_capsule
   center=<0 0.3759 -3.2011>
   rotation=<rot 83.3032 -90.003 -0.0029>
  }
  {shapeType=Capsule, length=15, radius=4, shapeName=MSTR_frontSkirt_capsule, center=<-5.8704 0.1069 -6.1943>, rotation=<rot 0.0002 0 -93.8167>}
 ]
 base_belt_l [{shapeType=Capsule, length=0.505, radius=0.505, shapeName=base_belt_l_capsule, rotation=<rot 0 -90 0>}]
 base_belt_r_ [{shapeType=Capsule, length=0.505, radius=0.505, shapeName=base_belt_r__capsule, rotation=<rot 90 0 0>}]
 base_buckle_back [{shapeType=Capsule, length=0.505, radius=0.505, shapeName=base_buckle_back_capsule, center=<0 -0.0493 0>}]
 base_buckle_front [{shapeType=Capsule, length=0.505, radius=0.505, shapeName=base_buckle_front_capsule, rotation=<rot 0 -90 0>}]
 dyn_belt_l_01 [{
   shapeType=Capsule
   length=4.1342
   radius=1.3288
   shapeName=dyn_frontBelt_01_l_capsule
   center=<3.1411 -1.2472 1.3063>
   rotation=<rot 90 -2.5426 0>
  }]
 dyn_belt_l_02 [{shapeType=Box, extentX=14.1511, extentY=1.5573, extentZ=5.5056, shapeName=dyn_belt_l_02_box, center=<7.7509 0 0>}]
 dyn_belt_l_03 [{shapeType=Box, extentX=8.9253, extentY=1.5573, extentZ=5.5056, shapeName=dyn_belt_l_03_box, center=<5.4168 0 0>}]
 dyn_belt_r_01 [{shapeType=Capsule, length=3.5196, radius=1.4791, shapeName=dyn_frontBelt_01_l_capsule, center=<3.063 0.24 0.0904>, rotation=<rot 90 0 0>}]
 dyn_belt_r_02 [{
   shapeType=Box
   extentX=6.4528
   extentY=0.9378
   extentZ=4.0651
   shapeName=dyn_belt_l_03_box
   center=<3.1674 0.0645 -0.0095>
   rotation=<rot 0 -0.0001 8.3759>
  }]
 dyn_buckle_back_01 [{shapeType=Capsule, length=3.1518, radius=1.0917, shapeName=dyn_buckle_back_01_capsule, center=<2.6586 -0.3659 0>, rotation=<rot 90 0 0>}]
 dyn_buckle_back_02 [{shapeType=Capsule, length=4.4954, radius=1.5, shapeName=dyn_buckle_back_02_capsule, center=<2.2439 -0.5852 0>, rotation=<rot 90 0 0>}]
 dyn_buckle_back_03 [{shapeType=Capsule, length=4.005, radius=1.505, shapeName=dyn_buckle_back_03_capsule, center=<2.7404 -0.7616 0>, rotation=<rot 90 0 0>}]
 dyn_buckle_back_04 [{shapeType=Capsule, length=4.4584, radius=1.3038, shapeName=dyn_buckle_back_04_capsule, center=<3.9468 -0.1148 0>, rotation=<rot 90 0 0>}]
 dyn_buckle_front_01 [{shapeType=Capsule, length=2, radius=1.5, shapeName=dyn_buckle_front_01_capsule, center=<2.1627 -0.9802 0>, rotation=<rot 90 0 0>}]
 dyn_buckle_front_02 [{shapeType=Capsule, length=1.8891, radius=1.4962, shapeName=dyn_buckle_front_02_capsule, center=<2.1757 -0.9402 0>, rotation=<rot 90 0 0>}]
 dyn_buckle_front_03 [{shapeType=Capsule, length=4.1214, radius=2, shapeName=dyn_buckle_front_03_capsule, center=<4.0857 -1.1655 0>, rotation=<rot 90 -0.0001 0>}]
 dyn_buckle_front_04 [{shapeType=Capsule, length=4.1001, radius=1.1611, shapeName=dyn_buckle_front_04_capsule, center=<4.0857 -1.3684 0>, rotation=<rot 90 0 0>}]
 dyn_frontSkirt_01 [{
   shapeType=Capsule
   length=11.7762
   radius=4.4359
   shapeName=dyn_frontSkirt_01_capsule
   center=<6.3587 -1.0059 -3.8832>
   rotation=<rot -0.0001 -179.7299 -84.7696>
  }]
 dyn_frontSkirt_01_l [{
   shapeType=Capsule
   length=9.2964
   radius=4.11
   shapeName=dyn_frontSkirt_01_l_capsule
   center=<7.0404 0.3079 -4.6684>
   rotation=<rot 0.0002 -1.7607 -76.4997>
  }]
 dyn_frontSkirt_01_r [{
   shapeType=Capsule
   length=9.0636
   radius=4
   shapeName=dyn_frontSkirt_01_r_capsule
   center=<6.4552 0.0033 -3.5588>
   rotation=<rot 0 -1.3814 -89.3343>
  }]
 dyn_frontSkirt_02 [
  {
   shapeType=Capsule
   length=12.2968
   radius=4.347
   shapeName=dyn_frontSkirt_02_capsule
   center=<3.7678 -1.1883 -3.8439>
   rotation=<rot -0.2884 179.9181 -88.3818>
  }
  {
   shapeType=Capsule
   length=10.6096
   radius=2.99
   shapeName=dyn_frontSkirt_02_capsule
   center=<9.3186 0.0065 -2.0642>
   rotation=<rot 0 0 -90.1801>
  }
 ]
 dyn_frontSkirt_02_l [{
   shapeType=Capsule
   length=10.9633
   radius=4.6843
   shapeName=dyn_frontSkirt_02_l_capsule
   center=<6.0469 2.9424 -5.3951>
   rotation=<rot -0.5501 -4.9099 -78.2807>
  }]
 dyn_frontSkirt_02_r [{
   shapeType=Capsule
   length=10.8737
   radius=3.5065
   shapeName=dyn_frontSkirt_02_r_capsule
   center=<4.7528 0.1356 -3.6052>
   rotation=<rot -2.5785 -179.1656 -89.9023>
  }]
 pelvis [{shapeType=Capsule, length=16.1321, radius=9.2893, shapeName=pelvis_capsule}]
 thigh_l [{shapeType=Capsule, length=24.7196, radius=10.5393, center=<-12.0871 -0.4997 0>, rotation=<rot 90 -5.0001 0>}]
 thigh_r [{shapeType=Capsule, length=18.143, radius=10.5, center=<12 0.3557 0>, rotation=<rot 90 0 0>}]
}
constraints [
 {
  bone1Name=base_belt_r_
  bone2Name=MSTR_Belt
  swing1LimitDegrees=45
  swing1Motion=Free
  swing2LimitDegrees=45
  swing2Motion=Free
  twistLimitDegrees=45
  twistMotion=Free
 }
 {bone1Name=dyn_belt_r_01, bone2Name=base_belt_r_, swing1LimitDegrees=20, swing2LimitDegrees=10, twistLimitDegrees=5}
 {bone1Name=dyn_belt_r_02, bone2Name=dyn_belt_r_01, swing1LimitDegrees=20, swing2LimitDegrees=10, twistLimitDegrees=5}
 {bone1Name=dyn_belt_l_01, bone2Name=base_belt_l, swing1LimitDegrees=20, swing2LimitDegrees=20, twistLimitDegrees=5}
 {bone1Name=dyn_belt_l_02, bone2Name=dyn_belt_l_01, swing1LimitDegrees=20, swing2LimitDegrees=20, twistLimitDegrees=5}
 {bone1Name=dyn_belt_l_03, bone2Name=dyn_belt_l_02, swing1LimitDegrees=20, swing2LimitDegrees=20, twistLimitDegrees=5}
 {bone1Name=dyn_buckle_back_01, bone2Name=base_buckle_back, swing1LimitDegrees=20, swing2LimitDegrees=5, twistLimitDegrees=5}
 {bone1Name=dyn_buckle_back_02, bone2Name=dyn_buckle_back_01, swing1LimitDegrees=20, swing2LimitDegrees=10, twistLimitDegrees=5}
 {bone1Name=dyn_buckle_back_03, bone2Name=dyn_buckle_back_02, swing1LimitDegrees=20, swing2LimitDegrees=20, twistLimitDegrees=5}
 {bone1Name=dyn_buckle_back_04, bone2Name=dyn_buckle_back_03, swing1LimitDegrees=20, swing2LimitDegrees=20, twistLimitDegrees=5}
 {bone1Name=dyn_buckle_front_01, bone2Name=base_buckle_front, swing1LimitDegrees=20, swing2LimitDegrees=5, twistLimitDegrees=5}
 {bone1Name=dyn_buckle_front_02, bone2Name=dyn_buckle_front_01, swing1LimitDegrees=20, swing2LimitDegrees=10, twistLimitDegrees=5}
 {bone1Name=dyn_buckle_front_03, bone2Name=dyn_buckle_front_02, swing1LimitDegrees=20, swing2LimitDegrees=10, twistLimitDegrees=5}
 {bone1Name=dyn_buckle_front_04, bone2Name=dyn_buckle_front_03, swing1LimitDegrees=10, swing2LimitDegrees=10, twistLimitDegrees=5}
 {
  bone1Name=thigh_l
  bone2Name=pelvis
  swing1LimitDegrees=45
  swing1Motion=Free
  swing2LimitDegrees=45
  swing2Motion=Free
  twistLimitDegrees=45
  twistMotion=Free
 }
 {
  bone1Name=thigh_r
  bone2Name=pelvis
  swing1LimitDegrees=45
  swing1Motion=Free
  swing2LimitDegrees=45
  swing2Motion=Free
  twistLimitDegrees=45
  twistMotion=Free
 }
 {
  bone1Name=MSTR_Belt
  bone2Name=pelvis
  swing1LimitDegrees=45
  swing1Motion=Free
  swing2LimitDegrees=45
  swing2Motion=Free
  twistLimitDegrees=45
  twistMotion=Free
 }
 {
  bone1Name=base_belt_l
  bone2Name=pelvis
  swing1LimitDegrees=45
  swing1Motion=Free
  swing2LimitDegrees=45
  swing2Motion=Free
  twistLimitDegrees=45
  twistMotion=Free
 }
 {
  bone1Name=base_buckle_front
  bone2Name=pelvis
  swing1LimitDegrees=45
  swing1Motion=Free
  swing2LimitDegrees=45
  swing2Motion=Free
  twistLimitDegrees=45
  twistMotion=Free
 }
 {
  bone1Name=base_buckle_back
  bone2Name=pelvis
  swing1LimitDegrees=45
  swing1Motion=Free
  swing2LimitDegrees=45
  swing2Motion=Free
  twistLimitDegrees=45
  twistMotion=Free
 }
 {
  bone1Name=dyn_frontSkirt_01
  bone2Name=pelvis
  swing1LimitDegrees=45
  swing1Motion=Free
  swing2LimitDegrees=45
  swing2Motion=Free
  twistLimitDegrees=45
  twistMotion=Free
 }
 {
  bone1Name=dyn_frontSkirt_02
  bone2Name=dyn_frontSkirt_01
  swing1LimitDegrees=45
  swing1Motion=Free
  swing2LimitDegrees=45
  swing2Motion=Free
  twistLimitDegrees=45
  twistMotion=Free
 }
 {
  bone1Name=MSTR_frontSkirt_l
  bone2Name=pelvis
  swing1LimitDegrees=45
  swing1Motion=Free
  swing2LimitDegrees=45
  swing2Motion=Free
  twistLimitDegrees=45
  twistMotion=Free
 }
 {
  bone1Name=dyn_frontSkirt_01_l
  bone2Name=MSTR_frontSkirt_l
  swing1LimitDegrees=45
  swing1Motion=Free
  swing2LimitDegrees=45
  swing2Motion=Free
  twistLimitDegrees=45
  twistMotion=Free
 }
 {
  bone1Name=dyn_frontSkirt_02_l
  bone2Name=dyn_frontSkirt_01_l
  swing1LimitDegrees=45
  swing1Motion=Free
  swing2LimitDegrees=45
  swing2Motion=Free
  twistLimitDegrees=45
  twistMotion=Free
 }
 {
  bone1Name=MSTR_frontSkirt_r
  bone2Name=pelvis
  swing1LimitDegrees=45
  swing1Motion=Free
  swing2LimitDegrees=45
  swing2Motion=Free
  twistLimitDegrees=45
  twistMotion=Free
 }
 {
  bone1Name=dyn_frontSkirt_01_r
  bone2Name=MSTR_frontSkirt_r
  swing1LimitDegrees=45
  swing1Motion=Free
  swing2LimitDegrees=45
  swing2Motion=Free
  twistLimitDegrees=45
  twistMotion=Free
 }
 {
  bone1Name=dyn_frontSkirt_02_r
  bone2Name=dyn_frontSkirt_01_r
  swing1LimitDegrees=45
  swing1Motion=Free
  swing2LimitDegrees=45
  swing2Motion=Free
  twistLimitDegrees=45
  twistMotion=Free
 }
]
