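# Physics asset for the M_MED_BlessFlan male body, covering the lower-body dangling parts: belts, buckles and front skirt.
# Pelvis and thigh capsules form the body core, with belt, buckle and three skirt chains constrained back to the pelvis, the right belt via MSTR_Belt.
# Belt and buckle links have tight angular limits, while skirt and root joints swing and twist freely.
# Several skirt bones stack multiple capsules to approximate wide panels.
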PhysicsAsset /BRCosmetics/Characters/Player/Male/Medium/Bodies/M_MED_BlessFlan/Meshes/M_MED_BlessFlan_Physics
bodies {
 MSTR_Belt [{shapeType=Capsule, length=0.505, radius=0.505, shapeName=MSTR_Belt_capsule, rotation=<rot 0 -90 0>}]
 MSTR_frontSkirt [
  {
   shapeType=Capsule
   length=12.2782
   radius=4.7218
   shapeName=MSTR_frontSkirt_capsule
   center=<-0.2479 -0.907 -4.2129>
   rotation=<rot 89.6785 -90.1815 -0.1815>
  }
  {
   shapeType=Capsule
   length=13.4133
   radius=4.6469
   shapeName=MSTR_frontSkirt_capsule
   center=<-6.0877 0.0136 -4.7964>
   rotation=<rot 0.0001 0 -90.1628>
  }
 ]
 MSTR_frontSkirt_l [
  {
   shapeType=Capsule
   length=10.7501
   radius=3.9994
   shapeName=MSTR_frontSkirt_l_capsule
   center=<-0.2016 -0.7889 -3.3399>
   rotation=<rot 76.4024 89.9485 -179.9995>
  }
  {
   shapeType=Capsule
   length=11.6574
   radius=4.4067
   shapeName=MSTR_frontSkirt_capsule
   center=<-5.8704 0.6012 -4.5694>
   rotation=<rot 0.0002 0 -82.7332>
  }
 ]
 MSTR_frontSkirt_r [
  {
   shapeType=Capsule
   length=8
   radius=4
   shapeName=MSTR_frontSkirt_r_capsule
   center=<0 0.2577 -2.1947>
   rotation=<rot 83.3032 -90.003 -0.0029>
  }
  {shapeType=Capsule, length=15, radius=4, shapeName=MSTR_frontSkirt_capsule, center=<-5.8704 0.0136 -4.7964>, rotation=<rot 0.0002 0 -93.8167>}
 ]
 base_belt_l [{shapeType=Capsule, length=0.505, radius=0.505, shapeName=base_belt_l_capsule, rotation=<rot 0 -90 0>}]
 base_belt_r_ [{shapeType=Capsule, length=0.505, radius=0.505, shapeName=base_belt_r__capsule, rotation=<rot 90 0 0>}]
 base_buckle_back [{shapeType=Capsule, length=0.505, radius=0.505, shapeName=base_buckle_back_capsule, center=<0 -0.0493 0>}]
 base_buckle_front [{shapeType=Capsule, length=0.505, radius=0.505, shapeName=base_buckle_front_capsule, rotation=<rot 0 -90 0>}]
 dyn_belt_l_01 [{shapeType=Capsule, length=5.6698, radius=1.3288, shapeName=dyn_frontBelt_01_l_capsule, center=<4.3389 -0.6486 0.4877>, rotation=<rot 90 0 0>}]
 dyn_belt_l_02 [{
   shapeType=Capsule
   length=8.7584
   radius=1.4999
   shapeName=dyn_frontBelt_01_l_capsule
   center=<6.9234 -0.2393 -0.4493>
   rotation=<rot 90 0 0>
  }]
 dyn_belt_l_03 [{
   shapeType=Capsule
   length=7.5328
   radius=1.7603
   shapeName=dyn_frontBelt_01_l_capsule
   center=<4.3492 -0.3419 -0.0004>
   rotation=<rot 88.2652 -0.5909 -2.2382>
  }]
 dyn_belt_r_01 [{shapeType=Capsule, length=3.5196, radius=1.4791, shapeName=dyn_frontBelt_01_l_capsule, center=<3.063 0.24 0.0904>, rotation=<rot 90 0 0>}]
 dyn_belt_r_02 [{shapeType=Capsule, length=4.2968, radius=1.7775, shapeName=dyn_frontBelt_01_l_capsule, center=<3.5004 0.4252 1.0827>, rotation=<rot 90 0 0>}]
 dyn_belt_r_03 [{
   shapeType=Capsule
   length=5.5922
   radius=1.7921
   shapeName=dyn_frontBelt_01_l_capsule
   center=<4.431 0.5151 1.0379>
   rotation=<rot 90 1.9428 0>
  }]
 dyn_buckle_back_01 [{shapeType=Capsule, length=3.1518, radius=1.0917, shapeName=dyn_buckle_back_01_capsule, center=<2.6586 -0.3659 0>, rotation=<rot 90 0 0>}]
 dyn_buckle_back_02 [{shapeType=Capsule, length=4.4954, radius=1.5, shapeName=dyn_buckle_back_02_capsule, center=<2.2439 -0.4 0>, rotation=<rot 90 0 0>}]
 dyn_buckle_back_03 [{shapeType=Capsule, length=4.005, radius=1.505, shapeName=dyn_buckle_back_03_capsule, center=<2.7404 0.094 0>, rotation=<rot 90 0 0>}]
 dyn_buckle_back_04 [{shapeType=Capsule, length=4.4584, radius=1.3038, shapeName=dyn_buckle_back_04_capsule, center=<3.9468 -0.1148 0>, rotation=<rot 90 0 0>}]
 dyn_buckle_front_01 [{shapeType=Capsule, length=2, radius=1.5, shapeName=dyn_buckle_front_01_capsule, center=<2.1627 -2 0>, rotation=<rot 90 0 0>}]
 dyn_buckle_front_02 [{shapeType=Capsule, length=1.8891, radius=1.4962, shapeName=dyn_buckle_front_02_capsule, center=<2.1757 -2 0>, rotation=<rot 90 0 0>}]
 dyn_buckle_front_03 [{shapeType=Capsule, length=4.6214, radius=1.5, shapeName=dyn_buckle_front_03_capsule, center=<4.0857 -1.8 0>, rotation=<rot 90 -4.7343 0>}]
 dyn_buckle_front_04 [{shapeType=Capsule, length=3.6001, radius=1.6611, shapeName=dyn_buckle_front_04_capsule, center=<4.0857 -2 0>, rotation=<rot 90 0 0>}]
 dyn_frontSkirt_01 [
  {
   shapeType=Capsule
   length=10.7762
   radius=5.4359
   shapeName=dyn_frontSkirt_01_capsule
   center=<6.3581 -0.8798 -5.3441>
   rotation=<rot -0.0003 -179.7299 -86.5696>
  }
  {
   shapeType=Capsule
   length=10.7762
   radius=5.4359
   shapeName=dyn_frontSkirt_01_capsule
   center=<1.9354 -0.9478 -5.3442>
   rotation=<rot -0.0003 -179.7299 -86.5696>
  }
 ]
 dyn_frontSkirt_01_l [
  {
   shapeType=Capsule
   length=9.626
   radius=5.11
   shapeName=dyn_frontSkirt_01_l_capsule
   center=<7.0741 1.4036 -4.9409>
   rotation=<rot 0.0002 -1.7607 -76.4997>
  }
  {
   shapeType=Capsule
   length=9.626
   radius=5.11
   shapeName=dyn_frontSkirt_01_l_capsule
   center=<12.3134 1.55 -4.9409>
   rotation=<rot 0.0002 -1.7607 -76.4997>
  }
 ]
 dyn_frontSkirt_01_r [{
   shapeType=Capsule
   length=9.2297
   radius=3.8339
   shapeName=dyn_frontSkirt_01_r_capsule
   center=<6.4552 0.0047 -3.4338>
   rotation=<rot 0 -1.3814 -89.3343>
  }]
 dyn_frontSkirt_02 [
  {
   shapeType=Capsule
   length=11.7968
   radius=4.847
   shapeName=dyn_frontSkirt_02_capsule
   center=<5.221 -1.1502 -5.2566>
   rotation=<rot -0.2884 179.9181 -88.3818>
  }
  {
   shapeType=Capsule
   length=12.2575
   radius=3.49
   shapeName=dyn_frontSkirt_02_capsule
   center=<9.3186 0.0091 -2.8876>
   rotation=<rot 0 0 -90.1801>
  }
  {
   shapeType=Capsule
   length=12.2575
   radius=3.49
   shapeName=dyn_frontSkirt_02_capsule
   center=<8.3265 -8.6662 -4.4605>
   rotation=<rot -5.6116 -11.7597 -110.4846>
  }
 ]
 dyn_frontSkirt_02_l [{
   shapeType=Capsule
   length=10.4339
   radius=5.6843
   shapeName=dyn_frontSkirt_02_l_capsule
   center=<6.04 2.8879 -5.6208>
   rotation=<rot -0.5502 -4.9099 -76.3251>
  }]
 dyn_frontSkirt_02_r [{
   shapeType=Capsule
   length=9.8737
   radius=4.5065
   shapeName=dyn_frontSkirt_02_r_capsule
   center=<4.7528 0.1356 -3.6052>
   rotation=<rot -2.5785 -179.1656 -89.9023>
  }]
 pelvis [{shapeType=Capsule, length=16.1321, radius=9.2893, shapeName=pelvis_capsule, center=<-0.294 0 0>}]
 thigh_l [{shapeType=Capsule, length=25.1992, radius=10.0596, center=<-12.0198 0.2695 0>, rotation=<rot 90 -5.0001 0>}]
 thigh_r [{shapeType=Capsule, length=20.7708, radius=9.3324, center=<12 0.3557 0>, rotation=<rot 90 0 0>}]
}
constraints [
 {
  bone1Name=base_belt_r_
  bone2Name=MSTR_Belt
  swing1LimitDegrees=45
  swing1Motion=Free
  swing2LimitDegrees=45
  swing2Motion=Free
  twistLimitDegrees=45
  twistMotion=Free
 }
 {bone1Name=dyn_belt_r_01, bone2Name=base_belt_r_, swing1LimitDegrees=20, swing2LimitDegrees=10, twistLimitDegrees=5}
 {bone1Name=dyn_belt_r_02, bone2Name=dyn_belt_r_01, swing1LimitDegrees=20, swing2LimitDegrees=10, twistLimitDegrees=5}
 {bone1Name=dyn_belt_r_03, bone2Name=dyn_belt_r_02, swing1LimitDegrees=20, swing2LimitDegrees=10, twistLimitDegrees=5}
 {bone1Name=dyn_belt_l_02, bone2Name=dyn_belt_l_01, swing1LimitDegrees=20, swing2LimitDegrees=20, twistLimitDegrees=5}
 {bone1Name=dyn_belt_l_03, bone2Name=dyn_belt_l_02, swing1LimitDegrees=20, swing2LimitDegrees=20, twistLimitDegrees=5}
 {bone1Name=dyn_buckle_back_01, bone2Name=base_buckle_back, swing1LimitDegrees=20, swing2LimitDegrees=5, twistLimitDegrees=5}
 {bone1Name=dyn_buckle_back_02, bone2Name=dyn_buckle_back_01, swing1LimitDegrees=20, swing2LimitDegrees=10, twistLimitDegrees=5}
 {bone1Name=dyn_buckle_back_03, bone2Name=dyn_buckle_back_02, swing1LimitDegrees=20, swing2LimitDegrees=20, twistLimitDegrees=5}
 {bone1Name=dyn_buckle_back_04, bone2Name=dyn_buckle_back_03, swing1LimitDegrees=20, swing2LimitDegrees=20, twistLimitDegrees=5}
 {bone1Name=dyn_buckle_front_01, bone2Name=base_buckle_front, swing1LimitDegrees=20, swing2LimitDegrees=5, twistLimitDegrees=5}
 {bone1Name=dyn_buckle_front_02, bone2Name=dyn_buckle_front_01, swing1LimitDegrees=20, swing2LimitDegrees=10, twistLimitDegrees=5}
 {bone1Name=dyn_buckle_front_03, bone2Name=dyn_buckle_front_02, swing1LimitDegrees=20, swing2LimitDegrees=10, twistLimitDegrees=5}
 {bone1Name=dyn_buckle_front_04, bone2Name=dyn_buckle_front_03, swing1LimitDegrees=10, swing2LimitDegrees=10, twistLimitDegrees=5}
 {
  bone1Name=thigh_l
  bone2Name=pelvis
  swing1LimitDegrees=45
  swing1Motion=Free
  swing2LimitDegrees=45
  swing2Motion=Free
  twistLimitDegrees=45
  twistMotion=Free
 }
 {
  bone1Name=thigh_r
  bone2Name=pelvis
  swing1LimitDegrees=45
  swing1Motion=Free
  swing2LimitDegrees=45
  swing2Motion=Free
  twistLimitDegrees=45
  twistMotion=Free
 }
 {
  bone1Name=MSTR_Belt
  bone2Name=pelvis
  swing1LimitDegrees=45
  swing1Motion=Free
  swing2LimitDegrees=45
  swing2Motion=Free
  twistLimitDegrees=45
  twistMotion=Free
 }
 {
  bone1Name=base_belt_l
  bone2Name=pelvis
  swing1LimitDegrees=45
  swing1Motion=Free
  swing2LimitDegrees=45
  swing2Motion=Free
  twistLimitDegrees=45
  twistMotion=Free
 }
 {
  bone1Name=base_buckle_front
  bone2Name=pelvis
  swing1LimitDegrees=45
  swing1Motion=Free
  swing2LimitDegrees=45
  swing2Motion=Free
  twistLimitDegrees=45
  twistMotion=Free
 }
 {
  bone1Name=base_buckle_back
  bone2Name=pelvis
  swing1LimitDegrees=45
  swing1Motion=Free
  swing2LimitDegrees=45
  swing2Motion=Free
  twistLimitDegrees=45
  twistMotion=Free
 }
 {
  bone1Name=MSTR_frontSkirt
  bone2Name=pelvis
  swing1LimitDegrees=45
  swing1Motion=Free
  swing2LimitDegrees=45
  swing2Motion=Free
  twistLimitDegrees=45
  twistMotion=Free
 }
 {
  bone1Name=dyn_frontSkirt_01
  bone2Name=MSTR_frontSkirt
  swing1LimitDegrees=45
  swing1Motion=Free
  swing2LimitDegrees=45
  swing2Motion=Free
  twistLimitDegrees=45
  twistMotion=Free
 }
 {
  bone1Name=dyn_frontSkirt_02
  bone2Name=dyn_frontSkirt_01
  swing1LimitDegrees=45
  swing1Motion=Free
  swing2LimitDegrees=45
  swing2Motion=Free
  twistLimitDegrees=45
  twistMotion=Free
 }
 {
  bone1Name=MSTR_frontSkirt_l
  bone2Name=pelvis
  swing1LimitDegrees=45
  swing1Motion=Free
  swing2LimitDegrees=45
  swing2Motion=Free
  twistLimitDegrees=45
  twistMotion=Free
 }
 {
  bone1Name=dyn_frontSkirt_01_l
  bone2Name=MSTR_frontSkirt_l
  swing1LimitDegrees=45
  swing1Motion=Free
  swing2LimitDegrees=45
  swing2Motion=Free
  twistLimitDegrees=45
  twistMotion=Free
 }
 {
  bone1Name=dyn_frontSkirt_02_l
  bone2Name=dyn_frontSkirt_01_l
  swing1LimitDegrees=45
  swing1Motion=Free
  swing2LimitDegrees=45
  swing2Motion=Free
  twistLimitDegrees=45
  twistMotion=Free
 }
 {
  bone1Name=MSTR_frontSkirt_r
  bone2Name=pelvis
  swing1LimitDegrees=45
  swing1Motion=Free
  swing2LimitDegrees=45
  swing2Motion=Free
  twistLimitDegrees=45
  twistMotion=Free
 }
 {
  bone1Name=dyn_frontSkirt_01_r
  bone2Name=MSTR_frontSkirt_r
  swing1LimitDegrees=45
  swing1Motion=Free
  swing2LimitDegrees=45
  swing2Motion=Free
  twistLimitDegrees=45
  twistMotion=Free
 }
 {
  bone1Name=dyn_frontSkirt_02_r
  bone2Name=dyn_frontSkirt_01_r
  swing1LimitDegrees=45
  swing1Motion=Free
  swing2LimitDegrees=45
  swing2Motion=Free
  twistLimitDegrees=45
  twistMotion=Free
 }
 {bone1Name=dyn_belt_l_01, bone2Name=base_belt_l, swing1LimitDegrees=20, swing2LimitDegrees=10, twistLimitDegrees=5}
]
