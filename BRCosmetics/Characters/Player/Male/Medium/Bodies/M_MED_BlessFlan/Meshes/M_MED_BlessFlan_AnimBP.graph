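# Post-process-style AnimBP for the BlessFlan male medium cosmetic body, adding secondary motion to skirts, belts, buckles and straps.
# It copies the parent mesh pose through a Control Rig, then applies weighted constraints and movement-driven bone offsets into cached ControlPose.
# AnimDynamics chains simulate skirt bones from that cache.
# Output branches from cached DynPose: RigidBody in the front end, otherwise cape-dependent buckle constraints or AnimDynamics belts, chosen by bRigidBodyEnabled.
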AnimBlueprint /BRCosmetics/Characters/Player/Male/Medium/Bodies/M_MED_BlessFlan/Meshes/M_MED_BlessFlan_AnimBP
target_skeleton=&/M_MED_BlessFlan_Skeleton
graphs [
 AnimGraph {nodes [
   n1 = Anim Dynamics - Bone: dyn_skirt_f_l!(SimulationSpace:AnimPhysSimSpaceType Enum=BoneRelative, ComponentPose=Local To Component(), Alpha=1, ExternalForce="0,0,0") -> (Pose) {
    angularDampingOverride=1
    angularSpringConstant=5000
    bAngularSpring=true
    bOverrideAngularDamping=true
    bUseGravityOverride=true
    gravityScale=0.35
    simulationSpace=BoneRelative
    alphaBoolBlend {blendOption=Linear}
    boundBone {boneName=dyn_skirt_f_l}
    componentAppliedLinearAccClamp=<3000 3000 3000>
    componentLinearAccScale=<1 1 1>
    componentLinearVelScale=<1 1 1>
    gravityOverride {z=-980}
    physicsBodyDefinitions [{
      boundBone {boneName=dyn_skirt_f_l}
      boxExtents=<20 20 20>
      constraintSetup {angularLimitsMax=<0 0 0>, angularLimitsMin=<0 0 -25>, angularTarget=<1 0 0>}
      localJointOffset=<4 0 0>
     }]
    relativeSpaceBone {boneName=spine_03}
   }
   n2 = Anim Dynamics - Bone: dyn_skirt_b_l!(SimulationSpace:AnimPhysSimSpaceType Enum=BoneRelative, ComponentPose=@n1, Alpha=1, ExternalForce="0,0,0") -> (Pose) {
    angularDampingOverride=1
    angularSpringConstant=5000
    bAngularSpring=true
    bOverrideAngularDamping=true
    bUseGravityOverride=true
    gravityScale=0.35
    simulationSpace=BoneRelative
    alphaBoolBlend {blendOption=Linear}
    boundBone {boneName=dyn_skirt_b_l}
    componentAppliedLinearAccClamp=<3000 3000 3000>
    componentLinearAccScale=<1 1 1>
    componentLinearVelScale=<1 1 1>
    gravityOverride {z=-980}
    physicsBodyDefinitions [{
      boundBone {boneName=dyn_skirt_b_l}
      boxExtents=<20 20 20>
      constraintSetup {angularLimitsMax=<0 0 25>, angularLimitsMin=<0 0 0>, angularTarget=<1 0 0>}
      localJointOffset=<4 0 0>
     }]
    relativeSpaceBone {boneName=spine_03}
   }
   n3 = Anim Dynamics - Bone: dyn_skirt_f_m!(SimulationSpace:AnimPhysSimSpaceType Enum=BoneRelative, ComponentPose=@n2, Alpha=1, ExternalForce="0,0,0") -> (Pose) {
    angularDampingOverride=1
    angularSpringConstant=5000
    bAngularSpring=true
    bOverrideAngularDamping=true
    bUseGravityOverride=true
    gravityScale=0.35
    simulationSpace=BoneRelative
    alphaBoolBlend {blendOption=Linear}
    boundBone {boneName=dyn_skirt_f_m}
    componentAppliedLinearAccClamp=<3000 3000 3000>
    componentLinearAccScale=<1 1 1>
    componentLinearVelScale=<1 1 1>
    gravityOverride {z=-980}
    physicsBodyDefinitions [{
      boundBone {boneName=dyn_skirt_f_m}
      boxExtents=<20 20 20>
      constraintSetup {angularLimitsMax=<0 0 0>, angularLimitsMin=<0 0 -25>, angularTarget=<1 0 0>}
      localJointOffset=<4 0 0>
     }]
    relativeSpaceBone {boneName=spine_03}
   }
   n4 = Anim Dynamics - Bone: dyn_skirt_b_m!(SimulationSpace:AnimPhysSimSpaceType Enum=BoneRelative, ComponentPose=@n3, Alpha=1, ExternalForce="0,0,0") -> (Pose) {
    angularDampingOverride=1
    angularSpringConstant=5000
    bAngularSpring=true
    bOverrideAngularDamping=true
    bUseGravityOverride=true
    gravityScale=0.35
    simulationSpace=BoneRelative
    alphaBoolBlend {blendOption=Linear}
    boundBone {boneName=dyn_skirt_b_m}
    componentAppliedLinearAccClamp=<3000 3000 3000>
    componentLinearAccScale=<1 1 1>
    componentLinearVelScale=<1 1 1>
    gravityOverride {z=-980}
    physicsBodyDefinitions [{
      boundBone {boneName=dyn_skirt_b_m}
      boxExtents=<20 20 20>
      constraintSetup {angularLimitsMax=<0 0 25>, angularLimitsMin=<0 0 0>, angularTarget=<1 0 0>}
      localJointOffset=<4 0 0>
     }]
    relativeSpaceBone {boneName=spine_03}
   }
   n5 = Anim Dynamics - Bone: dyn_skirt_f_r!(SimulationSpace:AnimPhysSimSpaceType Enum=BoneRelative, ComponentPose=@n4, Alpha=1, ExternalForce="0,0,0") -> (Pose) {
    angularDampingOverride=1
    angularSpringConstant=5000
    bAngularSpring=true
    bOverrideAngularDamping=true
    bUseGravityOverride=true
    gravityScale=0.2
    simulationSpace=BoneRelative
    alphaBoolBlend {blendOption=Linear}
    boundBone {boneName=dyn_skirt_f_r}
    componentAppliedLinearAccClamp=<3000 3000 3000>
    componentLinearAccScale=<1 1 1>
    componentLinearVelScale=<1 1 1>
    gravityOverride {z=-980}
    physicsBodyDefinitions [{
      boundBone {boneName=dyn_skirt_f_r}
      boxExtents=<20 20 20>
      constraintSetup {angularLimitsMax=<0 0 0>, angularLimitsMin=<0 0 -25>, angularTarget=<1 0 0>}
      localJointOffset=<4 0 0>
     }]
    relativeSpaceBone {boneName=spine_03}
   }
   Anim Dynamics - Bone: dyn_skirt_b_r!(SimulationSpace:AnimPhysSimSpaceType Enum=BoneRelative, ComponentPose=@n5, Alpha=1, ExternalForce="0,0,0") -> (Pose) {
    angularDampingOverride=1
    angularSpringConstant=5000
    bAngularSpring=true
    bOverrideAngularDamping=true
    bUseGravityOverride=true
    gravityScale=0.35
    simulationSpace=BoneRelative
    alphaBoolBlend {blendOption=Linear}
    boundBone {boneName=dyn_skirt_b_r}
    componentAppliedLinearAccClamp=<3000 3000 3000>
    componentLinearAccScale=<1 1 1>
    componentLinearVelScale=<1 1 1>
    gravityOverride {z=-980}
    physicsBodyDefinitions [{
      boundBone {boneName=dyn_skirt_b_r}
      boxExtents=<20 20 20>
      constraintSetup {angularLimitsMax=<0 0 25>, angularLimitsMin=<0 0 0>, angularTarget=<1 0 0>}
      localJointOffset=<3 0 0>
     }]
    relativeSpaceBone {boneName=spine_03}
   }
   Anim Dynamics - Bone: dyn_frontBelt_04_l!(ComponentPose, Alpha=1, GravityOverride:Vector="0,0,-980", ExternalForce="0,0,0") -> (Pose) {
    angularDampingOverride=0.9999
    angularSpringConstant=5000
    bAngularSpring=true
    bOverrideAngularDamping=true
    bUseGravityOverride=true
    bUsePlanarLimit=false
    gravityScale=0.35
    simulationSpace=BoneRelative
    boundBone {boneName=dyn_frontBelt_04_l}
    componentAppliedLinearAccClamp=<5000 5000 5000>
    componentLinearAccScale=<0.5 0.5 0.5>
    componentLinearVelScale=<0.5 0.5 0.5>
    gravityOverride {z=-980}
    physicsBodyDefinitions [{
      collisionType=CustomSphere
      sphereCollisionRadius=1
      boundBone {boneName=dyn_frontBelt_04_l}
      boxExtents=<20 20 20>
      constraintSetup {angularLimitsMax=<0 0 45>, angularLimitsMin=<0 -25 -45>, angularTarget=<1 0 0>}
      localJointOffset=<0.01 0 0>
     }]
    relativeSpaceBone {boneName=pelvis}
   }
   Component To Local()
   Component To Local()
   Component To Local()
   n6 = Get IsStandingNoMovement() -> (IsStandingNoMovement:bool)
   n7 = "Constraint - base_buckle_front (PPP)"!(ComponentPose, bAlphaBoolEnabled:bool=@n6, ConstraintWeights_0=0, ConstraintWeights_1:float=0, ConstraintWeights_2:float=0) -> (Pose) {
    alphaInputType=Bool
    boneToModify {boneName=base_buckle_front}
    constraintSetup [{targetBone {boneName=pelvis}}, {targetBone {boneName=thigh_r}}, {targetBone {boneName=thigh_twist_01_r}}]
    constraintWeights [1,1,1]
   }
   n8 = "Constraint - dyn_buckle_front_01 (PPP)"!(ComponentPose=@n7, bAlphaBoolEnabled:bool=@n6, ConstraintWeights_0=0, ConstraintWeights_1:float=0, ConstraintWeights_2:float=0) -> (Pose) {
    alphaInputType=Bool
    boneToModify {boneName=dyn_buckle_front_01}
    constraintSetup [{targetBone {boneName=pelvis}}, {targetBone {boneName=thigh_r}}, {targetBone {boneName=thigh_twist_01_r}}]
    constraintWeights [1,1,1]
   }
   n9 = "Constraint - dyn_buckle_front_02 (PPP)"!(ComponentPose=@n8, bAlphaBoolEnabled:bool=@n6, ConstraintWeights_0=0, ConstraintWeights_1:float=0, ConstraintWeights_2:float=0) -> (Pose) {
    alphaInputType=Bool
    boneToModify {boneName=dyn_buckle_front_02}
    constraintSetup [{targetBone {boneName=pelvis}}, {targetBone {boneName=thigh_r}}, {targetBone {boneName=thigh_twist_01_r}}]
    constraintWeights [1,1,1]
   }
   n10 = "Constraint - dyn_buckle_front_03 (PPP)"!(ComponentPose=@n9, Alpha=1, ConstraintWeights_0=0, ConstraintWeights_1:float=0, ConstraintWeights_2:float=0) -> (Pose) {
    boneToModify {boneName=dyn_buckle_front_03}
    constraintSetup [{targetBone {boneName=pelvis}}, {targetBone {boneName=thigh_r}}, {targetBone {boneName=thigh_twist_01_r}}]
    constraintWeights [1,1,1]
   }
   "Constraint - dyn_buckle_front_04 (PPP)"!(ComponentPose=@n10, Alpha=1, ConstraintWeights_0=0, ConstraintWeights_1:float=0, ConstraintWeights_2:float=0) -> (Pose) {
    boneToModify {boneName=dyn_buckle_front_04}
    constraintSetup [{targetBone {boneName=pelvis}}, {targetBone {boneName=thigh_r}}, {targetBone {boneName=thigh_twist_01_r}}]
    constraintWeights [1,1,1]
   }
   Local To Component()
   n11 = "Constraint - dyn_buckle_back_01 (PPPP)"!(
    ComponentPose=Local To Component(LocalPose=Use cached pose 'DynPose'())
    Alpha=1
    ConstraintWeights_0=1
    ConstraintWeights_1:float=0
    ConstraintWeights_2:float=0
    ConstraintWeights_3:float=0
   ) -> (Pose) {
    alphaBoolBlend {blendOption=Linear}
    boneToModify {boneName=dyn_buckle_back_01}
    constraintSetup [{targetBone {boneName=pelvis}}, {targetBone {boneName=spine_01}}, {targetBone {boneName=thigh_r}}, {targetBone {boneName=thigh_twist_01_r}}]
    constraintWeights [1,1,1,1]
   }
   n12 = "Constraint - dyn_buckle_back_02 (PPPP)"!(ComponentPose=@n11, Alpha=1, ConstraintWeights_0=1, ConstraintWeights_1:float=0, ConstraintWeights_2:float=0, ConstraintWeights_3:float=0) -> (Pose) {
    alphaBoolBlend {blendOption=Linear}
    boneToModify {boneName=dyn_buckle_back_02}
    constraintSetup [{targetBone {boneName=pelvis}}, {targetBone {boneName=spine_01}}, {targetBone {boneName=thigh_r}}, {targetBone {boneName=thigh_twist_01_r}}]
    constraintWeights [1,1,1,1]
   }
   n13 = "Constraint - dyn_buckle_back_03 (PPPP)"!(
    ComponentPose=@n12
    Alpha=1
    ConstraintWeights_0=0.8
    ConstraintWeights_1:float=0
    ConstraintWeights_2:float=0.2
    ConstraintWeights_3:float=0
   ) -> (Pose) {
    alphaBoolBlend {blendOption=Linear}
    boneToModify {boneName=dyn_buckle_back_03}
    constraintSetup [{targetBone {boneName=pelvis}}, {targetBone {boneName=spine_01}}, {targetBone {boneName=thigh_r}}, {targetBone {boneName=thigh_twist_01_r}}]
    constraintWeights [1,1,1,1]
   }
   n14 = "Constraint - dyn_buckle_back_04 (PPPP)"!(
    ComponentPose=@n13
    Alpha=1
    ConstraintWeights_0=0.8
    ConstraintWeights_1:float=0
    ConstraintWeights_2:float=0.2
    ConstraintWeights_3:float=0
   ) -> (Pose) {
    alphaBoolBlend {blendOption=Linear}
    boneToModify {boneName=dyn_buckle_back_04}
    constraintSetup [{targetBone {boneName=pelvis}}, {targetBone {boneName=spine_01}}, {targetBone {boneName=thigh_r}}, {targetBone {boneName=thigh_twist_01_r}}]
    constraintWeights [1,1,1,1]
   }
   n15 = Blend Poses by bool(
    BlendPose_0=Component To Local(ComponentPose=@n14)
    BlendPose_1=Use cached pose 'DynPose'()
    bActiveValue=Get is Wearing Cape() -> (is Wearing Cape:bool)
   )
   n16 = Anim Dynamics - Bone: dyn_belt_r_01!(ComponentPose=Local To Component(LocalPose=Use cached pose 'DynPose'()), Alpha=1, GravityOverride:Vector="0,0,-980", ExternalForce="0,0,0") -> (Pose) {
    angularDampingOverride=0.9999
    angularSpringConstant=5000
    bAngularSpring=true
    bOverrideAngularDamping=true
    bUseGravityOverride=true
    bUsePlanarLimit=false
    gravityScale=0.35
    simulationSpace=BoneRelative
    boundBone {boneName=dyn_belt_r_01}
    componentAppliedLinearAccClamp=<5000 5000 5000>
    componentLinearAccScale=<0.5 0.5 0.5>
    componentLinearVelScale=<0.5 0.5 0.5>
    gravityOverride {z=-980}
    physicsBodyDefinitions [{
      collisionType=CustomSphere
      sphereCollisionRadius=1
      boundBone {boneName=dyn_belt_r_01}
      boxExtents=<20 20 20>
      constraintSetup {angularLimitsMax=<0 0 45>, angularLimitsMin=<0 -25 -45>, angularTarget=<1 0 0>}
      localJointOffset=<0.01 0 0>
     }]
    relativeSpaceBone {boneName=pelvis}
   }
   n17 = Anim Dynamics - Bone: dyn_belt_r_02!(ComponentPose=@n16, Alpha=1, GravityOverride:Vector="0,0,-980", ExternalForce="0,0,0") -> (Pose) {
    angularDampingOverride=0.9999
    angularSpringConstant=5000
    bAngularSpring=true
    bOverrideAngularDamping=true
    bUseGravityOverride=true
    bUsePlanarLimit=false
    gravityScale=0.35
    simulationSpace=BoneRelative
    boundBone {boneName=dyn_belt_r_02}
    componentAppliedLinearAccClamp=<5000 5000 5000>
    componentLinearAccScale=<0.5 0.5 0.5>
    componentLinearVelScale=<0.5 0.5 0.5>
    gravityOverride {z=-980}
    physicsBodyDefinitions [{
      collisionType=CustomSphere
      sphereCollisionRadius=1
      boundBone {boneName=dyn_belt_r_02}
      boxExtents=<20 20 20>
      constraintSetup {angularLimitsMax=<0 0 45>, angularLimitsMin=<0 -25 -45>, angularTarget=<1 0 0>}
      localJointOffset=<0.01 0 0>
     }]
    relativeSpaceBone {boneName=pelvis}
   }
   n18 = Anim Dynamics - Bone: dyn_belt_r_03!(ComponentPose=@n17, Alpha=1, GravityOverride:Vector="0,0,-980", ExternalForce="0,0,0") -> (Pose) {
    angularDampingOverride=0.9999
    angularSpringConstant=5000
    bAngularSpring=true
    bOverrideAngularDamping=true
    bUseGravityOverride=true
    bUsePlanarLimit=false
    gravityScale=0.35
    simulationSpace=BoneRelative
    boundBone {boneName=dyn_belt_r_03}
    componentAppliedLinearAccClamp=<5000 5000 5000>
    componentLinearAccScale=<0.5 0.5 0.5>
    componentLinearVelScale=<0.5 0.5 0.5>
    gravityOverride {z=-980}
    physicsBodyDefinitions [{
      collisionType=CustomSphere
      sphereCollisionRadius=1
      boundBone {boneName=dyn_belt_r_03}
      boxExtents=<20 20 20>
      constraintSetup {angularLimitsMax=<0 0 45>, angularLimitsMin=<0 -25 -45>, angularTarget=<1 0 0>}
      localJointOffset=<0.01 0 0>
     }]
    relativeSpaceBone {boneName=pelvis}
   }
   n19 = Anim Dynamics - Bone: dyn_belt_l_01!(ComponentPose=@n18, Alpha=1, GravityOverride:Vector="0,0,-980", ExternalForce="0,0,0") -> (Pose) {
    angularDampingOverride=0.9999
    angularSpringConstant=5000
    bAngularSpring=true
    bOverrideAngularDamping=true
    bUseGravityOverride=true
    bUsePlanarLimit=false
    gravityScale=0.35
    simulationSpace=BoneRelative
    boundBone {boneName=dyn_belt_l_01}
    componentAppliedLinearAccClamp=<5000 5000 5000>
    componentLinearAccScale=<0.5 0.5 0.5>
    componentLinearVelScale=<0.5 0.5 0.5>
    gravityOverride {z=-980}
    physicsBodyDefinitions [{
      collisionType=CustomSphere
      sphereCollisionRadius=1
      boundBone {boneName=dyn_belt_l_01}
      boxExtents=<20 20 20>
      constraintSetup {angularLimitsMax=<0 0 45>, angularLimitsMin=<0 -25 -45>, angularTarget=<1 0 0>}
      localJointOffset=<0.01 0 0>
     }]
    relativeSpaceBone {boneName=pelvis}
   }
   n20 = Anim Dynamics - Bone: dyn_belt_l_02!(ComponentPose=@n19, Alpha=1, GravityOverride:Vector="0,0,-980", ExternalForce="0,0,0") -> (Pose) {
    angularDampingOverride=0.9999
    angularSpringConstant=5000
    bAngularSpring=true
    bOverrideAngularDamping=true
    bUseGravityOverride=true
    bUsePlanarLimit=false
    gravityScale=0.35
    simulationSpace=BoneRelative
    boundBone {boneName=dyn_belt_l_02}
    componentAppliedLinearAccClamp=<5000 5000 5000>
    componentLinearAccScale=<0.5 0.5 0.5>
    componentLinearVelScale=<0.5 0.5 0.5>
    gravityOverride {z=-980}
    physicsBodyDefinitions [{
      collisionType=CustomSphere
      sphereCollisionRadius=1
      boundBone {boneName=dyn_belt_l_02}
      boxExtents=<20 20 20>
      constraintSetup {angularLimitsMax=<0 0 45>, angularLimitsMin=<0 -25 -45>, angularTarget=<1 0 0>}
      localJointOffset=<0.01 0 0>
     }]
    relativeSpaceBone {boneName=pelvis}
   }
   n21 = Anim Dynamics - Bone: dyn_belt_l_03!(ComponentPose=@n20, Alpha=1, GravityOverride:Vector="0,0,-980", ExternalForce="0,0,0") -> (Pose) {
    angularDampingOverride=0.9999
    angularSpringConstant=5000
    bAngularSpring=true
    bOverrideAngularDamping=true
    bUseGravityOverride=true
    bUsePlanarLimit=false
    gravityScale=0.35
    simulationSpace=BoneRelative
    boundBone {boneName=dyn_belt_l_03}
    componentAppliedLinearAccClamp=<5000 5000 5000>
    componentLinearAccScale=<0.5 0.5 0.5>
    componentLinearVelScale=<0.5 0.5 0.5>
    gravityOverride {z=-980}
    physicsBodyDefinitions [{
      collisionType=CustomSphere
      sphereCollisionRadius=1
      boundBone {boneName=dyn_belt_l_03}
      boxExtents=<20 20 20>
      constraintSetup {angularLimitsMax=<0 0 45>, angularLimitsMin=<0 -25 -45>, angularTarget=<1 0 0>}
      localJointOffset=<0.01 0 0>
     }]
    relativeSpaceBone {boneName=pelvis}
   }
   n22 = "Transform (Modify) Bone - Bone: glute_l_mover_ASW"(ComponentPose=@n21, Alpha=1, Translation="4,0,0", Rotation="0,0,0", Scale="1,1,1") {translationMode=BMM_Additive, translationSpace=BCS_BoneSpace, boneToModify {boneName=glute_l_mover_ASW}, translation {x=4}}
   n23 = "Transform (Modify) Bone - Bone: glute_r_mover_ASW"(ComponentPose=@n22, Alpha=1, Translation="-3,0,0", Rotation="0,0,0", Scale="1,1,1") {translationMode=BMM_Additive, translationSpace=BCS_BoneSpace, boneToModify {boneName=glute_r_mover_ASW}, translation {x=-3}}
   n24 = Component To Local(ComponentPose=@n23)
   n25 = Get bRigidBodyEnabled()
   n26 = "Transform (Modify) Bone - Bone: MSTR_Belt"!(
    ComponentPose=Local To Component(LocalPose=Use cached pose 'DynPose'())
    bAlphaBoolEnabled:bool=True
    Translation="0,0,0"
    Rotation="-10,0,0"
   ) -> (Pose) {
    alphaInputType=Bool
    rotationMode=BMM_Additive
    rotationSpace=BCS_BoneSpace
    translationMode=BMM_Additive
    translationSpace=BCS_BoneSpace
    boneToModify {boneName=MSTR_Belt}
    rotation {pitch=-10}
   }
   n27 = "Transform (Modify) Bone - Bone: base_belt_l"!(ComponentPose=@n26, bAlphaBoolEnabled:bool=True, Translation="0,0,0", Rotation="-0.023,-19.620,0.019") -> (Pose) {
    alphaInputType=Bool
    rotationMode=BMM_Additive
    rotationSpace=BCS_BoneSpace
    translationMode=BMM_Additive
    translationSpace=BCS_BoneSpace
    boneToModify {boneName=base_belt_l}
    rotation=<rot -0.023 -19.62 0.019>
   }
   n28 = RigidBody!(ComponentPose=@n27, bAlphaBoolEnabled:bool, OverrideWorldGravity:Vector="0,0,0", ExternalForce) -> (Pose) {
    alphaInputType=Bool
    bClampLinearTranslationLimitToRefPose=true
    bOverrideWorldGravity=true
    overridePhysicsAsset=&/M_MED_BlessFlan_FrontEnd_Physics
    simulationSpace=BaseBoneSpace
    alphaBoolBlend {blendOption=Linear}
    baseBoneRef {boneName=pelvis}
    componentAppliedLinearAccClamp=<0 0 0>
    simSpaceSettings {
     maxAngularAcceleration=40
     maxAngularVelocity=8
     maxLinearAcceleration=3500
     maxLinearVelocity=700
     velocityScaleZ=0.5
     worldAlpha=1
     externalLinearDragV=<2 2 2>
    }
   }
   Output Pose(Result=Blend Poses by bool(
     BlendPose_0=Component To Local(ComponentPose=@n28)
     BlendPose_1=Blend Poses by bool(BlendPose_0=@n15, BlendPose_1=@n24, bActiveValue=@n25)
     bActiveValue=Get bIsInFrontEnd()
    )) {layerGroup=None}
   n29 = Control Rig!(
    Source=Copy Pose From Mesh() {bUseAttachedParent=true}
    Alpha
    PistolEquipped:bool=Get isPistolEquipped() -> (isPistolEquipped:bool)
    EmoteFloat:double=Get is Emote Float() -> (is Emote Float:double)
    Is DBNO:bool=Get bIsDBNO()
    PistolADSUp:double=Get Pistol UP() -> (Pistol UP:float)
    Is Crouching:bool=Get bIsCrouching()
    PistolADSDown:double=Get Pistol Down() -> (Pistol Down:float)
    Is Targeting:bool=Get TargetingNotPistol() -> (TargetingNotPistol:bool)
    Rigid Body Enabled:bool=Get bRigidBodyEnabled()
    Hang Gliding:bool=Get Hang Gliding() -> (Hang Gliding:bool)
    CrouchMoving:bool=Get IsCrouchMoving() -> (IsCrouchMoving:bool)
    Front End Idle:bool=Get Not Front End Idle() -> (Not Front End Idle:bool)
    Front End Idle Float:double=Get Not Front End Idle Float() -> (Not Front End Idle Float:double)
   ) -> (Pose) {controlRigAssetReference {blueprintRigClass=&/M_MED_BlessFlan_CtrlRig.M_MED_BlessFlan_CtrlRig_C}}
   n30 = "Constraint - base_pauldron_back_strap (PPPPPPPP)"!(
    ComponentPose=Local To Component(LocalPose=@n29)
    Alpha=1
    ConstraintWeights_0=0.3511
    ConstraintWeights_1:float=0.2251
    ConstraintWeights_2:float=0.1282
    ConstraintWeights_3:float=0.1159
    ConstraintWeights_4:float=0.0926
    ConstraintWeights_5:float=0.0417
    ConstraintWeights_6:float=0.024
    ConstraintWeights_7:float=0.0214
   ) -> (Pose) {
    boneToModify {boneName=base_pauldron_back_strap}
    constraintSetup [
     {targetBone {boneName=clavicle_r}}
     {targetBone {boneName=spine_03}}
     {targetBone {boneName=spine_04}}
     {targetBone {boneName=deform_upperarm_bk_r}}
     {targetBone {boneName=spine_05}}
     {targetBone {boneName=upperarm_twist_01_r}}
     {targetBone {boneName=spine_02}}
     {targetBone {boneName=deform_upperarm_up_r}}
    ]
    constraintWeights [1,1,1,1,1,1,1,1]
   }
   n31 = "Constraint - base_pauldron_front_strap (PPPPPPPP)"!(
    ComponentPose=@n30
    Alpha=1
    ConstraintWeights_0=0.2172
    ConstraintWeights_1:float=0.2142
    ConstraintWeights_2:float=0.175
    ConstraintWeights_3:float=0.1686
    ConstraintWeights_4:float=0.0819
    ConstraintWeights_5:float=0.0693
    ConstraintWeights_6:float=0.057
    ConstraintWeights_7:float=0.0168
   ) -> (Pose) {
    boneToModify {boneName=base_pauldron_front_strap}
    constraintSetup [
     {targetBone {boneName=spine_05}}
     {targetBone {boneName=clavicle_r}}
     {targetBone {boneName=spine_04}}
     {targetBone {boneName=deform_pec_r}}
     {targetBone {boneName=spine_03}}
     {targetBone {boneName=deform_upperarm_fr_r}}
     {targetBone {boneName=upperarm_twist_01_r}}
     {targetBone {boneName=upperarm_twist_02_r}}
    ]
    constraintWeights [1,1,1,1,1,1,1,1]
   }
   n32 = "Constraint - base_buckle_front (PPPPPPPP)"!(
    ComponentPose=@n31
    Alpha=1
    ConstraintWeights_0=0.7388
    ConstraintWeights_1:float=0.0778
    ConstraintWeights_2:float=0.0675
    ConstraintWeights_3:float=0.0491
    ConstraintWeights_4:float=0.0453
    ConstraintWeights_5:float=0.0191
    ConstraintWeights_6:float=0.0013
    ConstraintWeights_7:float=0.0011
   ) -> (Pose) {
    boneToModify {boneName=base_buckle_front}
    constraintSetup [
     {targetBone {boneName=pelvis}}
     {targetBone {boneName=spine_02}}
     {targetBone {boneName=spine_01}}
     {targetBone {boneName=deform_groin_r}}
     {targetBone {boneName=spine_03}}
     {targetBone {boneName=thigh_twist_01_r}}
     {targetBone {boneName=dyn_frontSkirt_01_r}}
     {targetBone {boneName=thigh_twist_01_l}}
    ]
    constraintWeights [1,1,1,1,1,1,1,1]
   }
   n33 = "Constraint - dyn_buckle_front_01 (PPPPPPPP)"!(
    ComponentPose=@n32
    Alpha=1
    ConstraintWeights_0=0.7388
    ConstraintWeights_1:float=0.0778
    ConstraintWeights_2:float=0.0675
    ConstraintWeights_3:float=0.0491
    ConstraintWeights_4:float=0.0453
    ConstraintWeights_5:float=0.0191
    ConstraintWeights_6:float=0.0013
    ConstraintWeights_7:float=0.0011
   ) -> (Pose) {
    boneToModify {boneName=dyn_buckle_front_01}
    constraintSetup [
     {targetBone {boneName=pelvis}}
     {targetBone {boneName=spine_02}}
     {targetBone {boneName=spine_01}}
     {targetBone {boneName=deform_groin_r}}
     {targetBone {boneName=spine_03}}
     {targetBone {boneName=thigh_twist_01_r}}
     {targetBone {boneName=dyn_frontSkirt_01_r}}
     {targetBone {boneName=thigh_twist_01_l}}
    ]
    constraintWeights [1,1,1,1,1,1,1,1]
   }
   n34 = "Constraint - dyn_buckle_front_02 (PPPPPPPP)"!(
    ComponentPose=@n33
    Alpha=1
    ConstraintWeights_0=0.6242
    ConstraintWeights_1:float=0.1873
    ConstraintWeights_2:float=0.0712
    ConstraintWeights_3:float=0.0706
    ConstraintWeights_4:float=0.029
    ConstraintWeights_5:float=0.0082
    ConstraintWeights_6:float=0.0054
    ConstraintWeights_7:float=0.0041
   ) -> (Pose) {
    boneToModify {boneName=dyn_buckle_front_02}
    constraintSetup [
     {targetBone {boneName=pelvis}}
     {targetBone {boneName=deform_groin_r}}
     {targetBone {boneName=thigh_twist_01_r}}
     {targetBone {boneName=dyn_frontSkirt_01_r}}
     {targetBone {boneName=dyn_frontSkirt_01}}
     {targetBone {boneName=dyn_frontSkirt_02_r}}
     {targetBone {boneName=thigh_twist_01_l}}
     {targetBone {boneName=dyn_sideSkirt_01_r}}
    ]
    constraintWeights [1,1,1,1,1,1,1,1]
   }
   n35 = "Constraint - dyn_buckle_front_03 (PPPPPPPP)"!(
    ComponentPose=@n34
    Alpha=1
    ConstraintWeights_0=0.4767
    ConstraintWeights_1:float=0.1654
    ConstraintWeights_2:float=0.1322
    ConstraintWeights_3:float=0.1068
    ConstraintWeights_4:float=0.0731
    ConstraintWeights_5:float=0.0294
    ConstraintWeights_6:float=0.0117
    ConstraintWeights_7:float=0.0047
   ) -> (Pose) {
    boneToModify {boneName=dyn_buckle_front_03}
    constraintSetup [
     {targetBone {boneName=pelvis}}
     {targetBone {boneName=dyn_frontSkirt_01_r}}
     {targetBone {boneName=deform_groin_r}}
     {targetBone {boneName=thigh_twist_01_r}}
     {targetBone {boneName=dyn_frontSkirt_01}}
     {targetBone {boneName=dyn_frontSkirt_02_r}}
     {targetBone {boneName=dyn_sideSkirt_01_r}}
     {targetBone {boneName=thigh_twist_01_l}}
    ]
    constraintWeights [1,1,1,1,1,1,1,1]
   }
   n36 = "Constraint - dyn_buckle_front_04 (PPPPPPPP)"!(
    ComponentPose=@n35
    Alpha=1
    ConstraintWeights_0=0.232
    ConstraintWeights_1:float=0.2067
    ConstraintWeights_2:float=0.1629
    ConstraintWeights_3:float=0.118
    ConstraintWeights_4:float=0.108
    ConstraintWeights_5:float=0.1068
    ConstraintWeights_6:float=0.0575
    ConstraintWeights_7:float=0.0082
   ) -> (Pose) {
    boneToModify {boneName=dyn_buckle_front_04}
    constraintSetup [
     {targetBone {boneName=dyn_frontSkirt_01_r}}
     {targetBone {boneName=pelvis}}
     {targetBone {boneName=dyn_frontSkirt_01}}
     {targetBone {boneName=dyn_frontSkirt_02_r}}
     {targetBone {boneName=deform_groin_r}}
     {targetBone {boneName=thigh_twist_01_r}}
     {targetBone {boneName=dyn_frontSkirt_02}}
     {targetBone {boneName=dyn_sideSkirt_02_r}}
    ]
    constraintWeights [1,1,1,1,1,1,1,1]
   }
   n37 = "Transform (Modify) Bone - Bone: base_buckle_front"(ComponentPose=@n36, Alpha=Get StandingForward() -> (StandingForward:double), Translation="0,0,0", Rotation="90,0,0", Scale="1,1,1") {
    rotationMode=BMM_Additive
    rotationSpace=BCS_BoneSpace
    scaleMode=BMM_Additive
    scaleSpace=BCS_BoneSpace
    translationMode=BMM_Additive
    translationSpace=BCS_BoneSpace
    alphaScaleBiasClamp {bInterpResult=true, interpSpeedDecreasing=0.1, interpSpeedIncreasing=0.1}
    boneToModify {boneName=base_buckle_front}
    rotation {pitch=90}
   }
   n38 = "Transform (Modify) Bone - Bone: base_buckle_front"!(
    ComponentPose=@n37
    bAlphaBoolEnabled:bool=Get Moving or Crouching() -> (Moving or Crouching:bool)
    Translation="0,0,0"
    Rotation="0,0,0"
    Scale="1,1,1"
   ) -> (Pose) {
    alphaInputType=Bool
    rotationMode=BMM_Additive
    rotationSpace=BCS_BoneSpace
    scaleMode=BMM_Additive
    scaleSpace=BCS_BoneSpace
    translationMode=BMM_Additive
    translationSpace=BCS_BoneSpace
    boneToModify {boneName=base_buckle_front}
   }
   n39 = Get Moving or Crouching()
   n40 = "Transform (Modify) Bone - Bone: base_buckle_back"!(ComponentPose=@n38, bAlphaBoolEnabled:bool=@n39, Translation="0,0,0", Rotation="0,-15,0", Scale="1,1,1") -> (Pose) {
    alphaInputType=Bool
    rotationMode=BMM_Additive
    rotationSpace=BCS_BoneSpace
    scaleMode=BMM_Additive
    scaleSpace=BCS_BoneSpace
    translationMode=BMM_Additive
    translationSpace=BCS_BoneSpace
    boneToModify {boneName=base_buckle_back}
    rotation {yaw=-15}
   }
   n41 = "Transform (Modify) Bone - Bone: MSTR_Belt"!(ComponentPose=@n40, bAlphaBoolEnabled=@n39, Translation="0,0,0", Rotation="0,0,0", Scale:Vector="1,1,1") -> (Pose) {
    alphaInputType=Bool
    rotationMode=BMM_Additive
    rotationSpace=BCS_BoneSpace
    translationMode=BMM_Additive
    translationSpace=BCS_BoneSpace
    alphaScaleBiasClamp {bInterpResult=true, interpSpeedDecreasing=1, interpSpeedIncreasing=1}
    boneToModify {boneName=MSTR_Belt}
   }
   n42 = "Transform (Modify) Bone - Bone: MSTR_Belt"!(ComponentPose=@n41, Alpha:float=1, Translation="0,0,0", Rotation="0,0,0", Scale="1,1,1") -> (Pose) {
    rotationMode=BMM_Additive
    rotationSpace=BCS_BoneSpace
    translationMode=BMM_Additive
    translationSpace=BCS_BoneSpace
    alphaScaleBiasClamp {bInterpResult=true, interpSpeedDecreasing=1, interpSpeedIncreasing=1}
    boneToModify {boneName=MSTR_Belt}
   }
   n43 = "Transform (Modify) Bone - Bone: base_buckle_front"!(ComponentPose=@n42, Alpha:float=Get RightMovement() -> (RightMovement:double), Translation="0,0,0", Rotation="30,0,0", Scale="1,1,1") -> (Pose) {
    rotationMode=BMM_Additive
    rotationSpace=BCS_BoneSpace
    scaleMode=BMM_Additive
    scaleSpace=BCS_BoneSpace
    translationMode=BMM_Additive
    translationSpace=BCS_BoneSpace
    boneToModify {boneName=base_buckle_front}
    rotation {pitch=30}
   }
   n44 = "Transform (Modify) Bone - Bone: base_buckle_front"(ComponentPose=@n43, Alpha=Get ForwardMovement() -> (ForwardMovement:double), Translation="0,0,0", Rotation="29.987,0,0.037", Scale="1,1,1") {
    rotationMode=BMM_Additive
    rotationSpace=BCS_BoneSpace
    scaleMode=BMM_Additive
    scaleSpace=BCS_BoneSpace
    translationMode=BMM_Additive
    translationSpace=BCS_BoneSpace
    boneToModify {boneName=base_buckle_front}
    rotation {pitch=29.987, roll=0.037}
   }
   n45 = "Transform (Modify) Bone - Bone: base_buckle_front"!(ComponentPose=@n44, bAlphaBoolEnabled:bool=Get IsCrouchMoving(), Translation="0,0,0", Rotation="0,-10,0", Scale="1,1,1") -> (Pose) {
    alphaInputType=Bool
    rotationMode=BMM_Additive
    rotationSpace=BCS_BoneSpace
    scaleMode=BMM_Additive
    scaleSpace=BCS_BoneSpace
    translationMode=BMM_Additive
    translationSpace=BCS_BoneSpace
    alphaBoolBlend {blendInTime=1, blendOutTime=1}
    boneToModify {boneName=base_buckle_front}
    rotation {yaw=-10}
   }
   Save cached pose 'ControlPose'(Pose=Component To Local(ComponentPose=@n45)) {cacheName=ControlPose}
   n46 = Anim Dynamics - Bone: dyn_frontSkirt_01_l!(
    SimulationSpace:AnimPhysSimSpaceType Enum
    ComponentPose=Local To Component(LocalPose=Use cached pose 'ControlPose'())
    Alpha=1
    GravityOverride:Vector="0,0,-980"
    ExternalForce="0,0,0"
   ) -> (Pose) {
    angularDampingOverride=1
    angularSpringConstant=20000
    bAngularSpring=true
    bOverrideAngularDamping=true
    bUseGravityOverride=true
    gravityScale=0.35
    simulationSpace=BoneRelative
    boundBone {boneName=dyn_frontSkirt_01_l}
    componentAppliedLinearAccClamp=<500 500 500>
    componentLinearAccScale=<0.1 0.1 0.1>
    componentLinearVelScale=<0.2 0.2 0.2>
    gravityOverride {z=-980}
    physicsBodyDefinitions [{
      boundBone {boneName=dyn_frontSkirt_01_l}
      boxExtents=<20 30 30>
      constraintSetup {angularLimitsMax=<0 0 0>, angularLimitsMin=<0 -3 0>, angularTarget=<1 0 0>}
      localJointOffset=<4 0 0>
     }]
    relativeSpaceBone {boneName=root}
    simSpaceSettings {maxAngularAcceleration=10, maxAngularVelocity=10}
   }
   n47 = Anim Dynamics - Bone: dyn_frontSkirt_02_l!(SimulationSpace:AnimPhysSimSpaceType Enum, ComponentPose=@n46, Alpha=1, GravityOverride:Vector="0,0,-980", ExternalForce="0,0,0") -> (Pose) {
    angularDampingOverride=1
    angularSpringConstant=20000
    bAngularSpring=true
    bOverrideAngularDamping=true
    bUseGravityOverride=true
    gravityScale=0.35
    simulationSpace=BoneRelative
    boundBone {boneName=dyn_frontSkirt_02_l}
    componentAppliedLinearAccClamp=<500 500 500>
    componentLinearAccScale=<0.1 0.1 0.1>
    componentLinearVelScale=<0.2 0.2 0.2>
    gravityOverride {z=-980}
    physicsBodyDefinitions [{
      boundBone {boneName=dyn_frontSkirt_02_l}
      boxExtents=<20 30 30>
      constraintSetup {angularLimitsMax=<0 0 0>, angularLimitsMin=<0 -10 0>, angularTarget=<1 0 0>}
      localJointOffset=<4 0 0>
     }]
    relativeSpaceBone {boneName=root}
    simSpaceSettings {maxAngularAcceleration=10, maxAngularVelocity=10}
   }
   n48 = Anim Dynamics - Bone: dyn_sideSkirt_01_l!(SimulationSpace:AnimPhysSimSpaceType Enum, ComponentPose=@n47, Alpha=1, GravityOverride:Vector="0,0,-980", ExternalForce="0,0,0") -> (Pose) {
    angularDampingOverride=1
    angularSpringConstant=20000
    bAngularSpring=true
    bOverrideAngularDamping=true
    bUseGravityOverride=true
    gravityScale=0.35
    simulationSpace=BoneRelative
    boundBone {boneName=dyn_sideSkirt_01_l}
    componentAppliedLinearAccClamp=<500 500 500>
    componentLinearAccScale=<0.1 0.1 0.1>
    componentLinearVelScale=<0.2 0.2 0.2>
    gravityOverride {z=-980}
    physicsBodyDefinitions [{
      boundBone {boneName=dyn_sideSkirt_01_l}
      boxExtents=<20 30 30>
      constraintSetup {angularLimitsMax=<0 0 0>, angularLimitsMin=<0 -3 0>, angularTarget=<1 0 0>}
      localJointOffset=<4 0 0>
     }]
    relativeSpaceBone {boneName=root}
    simSpaceSettings {maxAngularAcceleration=10, maxAngularVelocity=10}
   }
   n49 = Anim Dynamics - Bone: dyn_sideSkirt_02_l!(SimulationSpace:AnimPhysSimSpaceType Enum, ComponentPose=@n48, Alpha=1, GravityOverride:Vector="0,0,-980", ExternalForce="0,0,0") -> (Pose) {
    angularDampingOverride=1
    angularSpringConstant=20000
    bAngularSpring=true
    bOverrideAngularDamping=true
    bUseGravityOverride=true
    gravityScale=0.35
    simulationSpace=BoneRelative
    boundBone {boneName=dyn_sideSkirt_02_l}
    componentAppliedLinearAccClamp=<500 500 500>
    componentLinearAccScale=<0.1 0.1 0.1>
    componentLinearVelScale=<0.2 0.2 0.2>
    gravityOverride {z=-980}
    physicsBodyDefinitions [{
      boundBone {boneName=dyn_sideSkirt_02_l}
      boxExtents=<20 30 30>
      constraintSetup {angularLimitsMax=<0 0 0>, angularLimitsMin=<0 -10 0>, angularTarget=<1 0 0>}
      localJointOffset=<4 0 0>
     }]
    relativeSpaceBone {boneName=root}
    simSpaceSettings {maxAngularAcceleration=10, maxAngularVelocity=10}
   }
   n50 = Anim Dynamics - Bone: dyn_backSkirt_01_l!(SimulationSpace:AnimPhysSimSpaceType Enum, ComponentPose=@n49, Alpha=1, GravityOverride:Vector="0,0,-980", ExternalForce="0,0,0") -> (Pose) {
    angularDampingOverride=1
    angularSpringConstant=20000
    bAngularSpring=true
    bOverrideAngularDamping=true
    bUseGravityOverride=true
    gravityScale=0.35
    simulationSpace=BoneRelative
    boundBone {boneName=dyn_backSkirt_01_l}
    componentAppliedLinearAccClamp=<500 500 500>
    componentLinearAccScale=<0.1 0.1 0.1>
    componentLinearVelScale=<0.2 0.2 0.2>
    gravityOverride {z=-980}
    physicsBodyDefinitions [{
      boundBone {boneName=dyn_backSkirt_01_l}
      boxExtents=<20 30 30>
      constraintSetup {angularLimitsMax=<0 0 0>, angularLimitsMin=<0 -3 0>, angularTarget=<1 0 0>}
      localJointOffset=<4 0 0>
     }]
    relativeSpaceBone {boneName=root}
    simSpaceSettings {maxAngularAcceleration=10, maxAngularVelocity=10}
   }
   n51 = Anim Dynamics - Bone: dyn_backSkirt_02_l!(SimulationSpace:AnimPhysSimSpaceType Enum, ComponentPose=@n50, Alpha=1, GravityOverride:Vector="0,0,-980", ExternalForce="0,0,0") -> (Pose) {
    angularDampingOverride=1
    angularSpringConstant=20000
    bAngularSpring=true
    bOverrideAngularDamping=true
    bUseGravityOverride=true
    gravityScale=0.35
    simulationSpace=BoneRelative
    boundBone {boneName=dyn_backSkirt_02_l}
    componentAppliedLinearAccClamp=<500 500 500>
    componentLinearAccScale=<0.1 0.1 0.1>
    componentLinearVelScale=<0.2 0.2 0.2>
    gravityOverride {z=-980}
    physicsBodyDefinitions [{
      boundBone {boneName=dyn_backSkirt_02_l}
      boxExtents=<20 30 30>
      constraintSetup {angularLimitsMax=<0 0 0>, angularLimitsMin=<0 -10 0>, angularTarget=<1 0 0>}
      localJointOffset=<4 0 0>
     }]
    relativeSpaceBone {boneName=root}
    simSpaceSettings {maxAngularAcceleration=10, maxAngularVelocity=10}
   }
   n52 = Anim Dynamics - Bone: dyn_frontSkirt_01_r!(SimulationSpace:AnimPhysSimSpaceType Enum, ComponentPose=@n51, Alpha=1, GravityOverride:Vector="0,0,-980", ExternalForce="0,0,0") -> (Pose) {
    angularDampingOverride=1
    angularSpringConstant=20000
    bAngularSpring=true
    bOverrideAngularDamping=true
    bUseGravityOverride=true
    gravityScale=0.35
    simulationSpace=BoneRelative
    boundBone {boneName=dyn_frontSkirt_01_r}
    componentAppliedLinearAccClamp=<500 500 500>
    componentLinearAccScale=<0.1 0.1 0.1>
    componentLinearVelScale=<0.2 0.2 0.2>
    gravityOverride {z=-980}
    physicsBodyDefinitions [{
      boundBone {boneName=dyn_frontSkirt_01_r}
      boxExtents=<20 30 30>
      constraintSetup {angularLimitsMax=<0 0 0>, angularLimitsMin=<0 -3 0>, angularTarget=<1 0 0>}
      localJointOffset=<4 0 0>
     }]
    relativeSpaceBone {boneName=root}
    simSpaceSettings {maxAngularAcceleration=10, maxAngularVelocity=10}
   }
   n53 = Anim Dynamics - Bone: dyn_frontSkirt_02_r!(SimulationSpace:AnimPhysSimSpaceType Enum, ComponentPose=@n52, Alpha=1, GravityOverride:Vector="0,0,-980", ExternalForce="0,0,0") -> (Pose) {
    angularDampingOverride=1
    angularSpringConstant=20000
    bAngularSpring=true
    bOverrideAngularDamping=true
    bUseGravityOverride=true
    gravityScale=0.35
    simulationSpace=BoneRelative
    boundBone {boneName=dyn_frontSkirt_02_r}
    componentAppliedLinearAccClamp=<500 500 500>
    componentLinearAccScale=<0.1 0.1 0.1>
    componentLinearVelScale=<0.2 0.2 0.2>
    gravityOverride {z=-980}
    physicsBodyDefinitions [{
      boundBone {boneName=dyn_frontSkirt_02_r}
      boxExtents=<20 30 30>
      constraintSetup {angularLimitsMax=<0 0 0>, angularLimitsMin=<0 -10 0>, angularTarget=<1 0 0>}
      localJointOffset=<4 0 0>
     }]
    relativeSpaceBone {boneName=root}
    simSpaceSettings {maxAngularAcceleration=10, maxAngularVelocity=10}
   }
   n54 = Anim Dynamics - Bone: dyn_sideSkirt_01_r!(SimulationSpace:AnimPhysSimSpaceType Enum, ComponentPose=@n53, Alpha=1, GravityOverride:Vector="0,0,-980", ExternalForce="0,0,0") -> (Pose) {
    angularDampingOverride=1
    angularSpringConstant=20000
    bAngularSpring=true
    bOverrideAngularDamping=true
    bUseGravityOverride=true
    gravityScale=0.35
    simulationSpace=BoneRelative
    boundBone {boneName=dyn_sideSkirt_01_r}
    componentAppliedLinearAccClamp=<500 500 500>
    componentLinearAccScale=<0.1 0.1 0.1>
    componentLinearVelScale=<0.2 0.2 0.2>
    gravityOverride {z=-980}
    physicsBodyDefinitions [{
      boundBone {boneName=dyn_sideSkirt_01_r}
      boxExtents=<20 30 30>
      constraintSetup {angularLimitsMax=<0 0 0>, angularLimitsMin=<0 -3 0>, angularTarget=<1 0 0>}
      localJointOffset=<4 0 0>
     }]
    relativeSpaceBone {boneName=root}
    simSpaceSettings {maxAngularAcceleration=10, maxAngularVelocity=10}
   }
   n55 = Anim Dynamics - Bone: dyn_sideSkirt_02_r!(SimulationSpace:AnimPhysSimSpaceType Enum, ComponentPose=@n54, Alpha=1, GravityOverride:Vector="0,0,-980", ExternalForce="0,0,0") -> (Pose) {
    angularDampingOverride=1
    angularSpringConstant=20000
    bAngularSpring=true
    bOverrideAngularDamping=true
    bUseGravityOverride=true
    gravityScale=0.35
    simulationSpace=BoneRelative
    boundBone {boneName=dyn_sideSkirt_02_r}
    componentAppliedLinearAccClamp=<500 500 500>
    componentLinearAccScale=<0.1 0.1 0.1>
    componentLinearVelScale=<0.2 0.2 0.2>
    gravityOverride {z=-980}
    physicsBodyDefinitions [{
      boundBone {boneName=dyn_sideSkirt_02_r}
      boxExtents=<20 30 30>
      constraintSetup {angularLimitsMax=<0 0 0>, angularLimitsMin=<0 -10 0>, angularTarget=<1 0 0>}
      localJointOffset=<4 0 0>
     }]
    relativeSpaceBone {boneName=root}
    simSpaceSettings {maxAngularAcceleration=10, maxAngularVelocity=10}
   }
   n56 = Anim Dynamics - Bone: dyn_backSkirt_01_r!(SimulationSpace:AnimPhysSimSpaceType Enum, ComponentPose=@n55, Alpha=1, GravityOverride:Vector="0,0,-980", ExternalForce="0,0,0") -> (Pose) {
    angularDampingOverride=1
    angularSpringConstant=20000
    bAngularSpring=true
    bOverrideAngularDamping=true
    bUseGravityOverride=true
    gravityScale=0.35
    simulationSpace=BoneRelative
    boundBone {boneName=dyn_backSkirt_01_r}
    componentAppliedLinearAccClamp=<500 500 500>
    componentLinearAccScale=<0.1 0.1 0.1>
    componentLinearVelScale=<0.2 0.2 0.2>
    gravityOverride {z=-980}
    physicsBodyDefinitions [{
      boundBone {boneName=dyn_backSkirt_01_r}
      boxExtents=<20 30 30>
      constraintSetup {angularLimitsMax=<0 0 0>, angularLimitsMin=<0 -3 0>, angularTarget=<1 0 0>}
      localJointOffset=<4 0 0>
     }]
    relativeSpaceBone {boneName=root}
    simSpaceSettings {maxAngularAcceleration=10, maxAngularVelocity=10}
   }
   n57 = Anim Dynamics - Bone: dyn_backSkirt_02_r!(SimulationSpace:AnimPhysSimSpaceType Enum, ComponentPose=@n56, Alpha=1, GravityOverride:Vector="0,0,-980", ExternalForce="0,0,0") -> (Pose) {
    angularDampingOverride=1
    angularSpringConstant=20000
    bAngularSpring=true
    bOverrideAngularDamping=true
    bUseGravityOverride=true
    gravityScale=0.35
    simulationSpace=BoneRelative
    boundBone {boneName=dyn_backSkirt_02_r}
    componentAppliedLinearAccClamp=<500 500 500>
    componentLinearAccScale=<0.1 0.1 0.1>
    componentLinearVelScale=<0.2 0.2 0.2>
    gravityOverride {z=-980}
    physicsBodyDefinitions [{
      boundBone {boneName=dyn_backSkirt_02_r}
      boxExtents=<20 30 30>
      constraintSetup {angularLimitsMax=<0 0 0>, angularLimitsMin=<0 -10 0>, angularTarget=<1 0 0>}
      localJointOffset=<4 0 0>
     }]
    relativeSpaceBone {boneName=root}
    simSpaceSettings {maxAngularAcceleration=10, maxAngularVelocity=10}
   }
   n58 = Anim Dynamics - Bone: dyn_frontSkirt_01!(SimulationSpace:AnimPhysSimSpaceType Enum, ComponentPose=@n57, Alpha=1, GravityOverride:Vector="0,0,-980", ExternalForce="0,0,0") -> (Pose) {
    angularDampingOverride=1
    angularSpringConstant=20000
    bAngularSpring=true
    bOverrideAngularDamping=true
    bUseGravityOverride=true
    gravityScale=0.35
    simulationSpace=BoneRelative
    boundBone {boneName=dyn_frontSkirt_01}
    componentAppliedLinearAccClamp=<500 500 500>
    componentLinearAccScale=<0.1 0.1 0.1>
    componentLinearVelScale=<0.2 0.2 0.2>
    gravityOverride {z=-980}
    physicsBodyDefinitions [{
      boundBone {boneName=dyn_frontSkirt_01}
      boxExtents=<20 30 30>
      constraintSetup {angularLimitsMax=<0 0 0>, angularLimitsMin=<0 -3 0>, angularTarget=<1 0 0>}
      localJointOffset=<4 0 0>
     }]
    relativeSpaceBone {boneName=root}
    simSpaceSettings {maxAngularAcceleration=10, maxAngularVelocity=10}
   }
   n59 = Anim Dynamics - Bone: dyn_frontSkirt_02!(SimulationSpace:AnimPhysSimSpaceType Enum, ComponentPose=@n58, Alpha=1, GravityOverride:Vector="0,0,-980", ExternalForce="0,0,0") -> (Pose) {
    angularDampingOverride=1
    angularSpringConstant=20000
    bAngularSpring=true
    bOverrideAngularDamping=true
    bUseGravityOverride=true
    gravityScale=0.35
    simulationSpace=BoneRelative
    boundBone {boneName=dyn_frontSkirt_02}
    componentAppliedLinearAccClamp=<500 500 500>
    componentLinearAccScale=<0.1 0.1 0.1>
    componentLinearVelScale=<0.2 0.2 0.2>
    gravityOverride {z=-980}
    physicsBodyDefinitions [{
      boundBone {boneName=dyn_frontSkirt_02}
      boxExtents=<20 30 30>
      constraintSetup {angularLimitsMax=<0 0 0>, angularLimitsMin=<0 -10 0>, angularTarget=<1 0 0>}
      localJointOffset=<4 0 0>
     }]
    relativeSpaceBone {boneName=root}
    simSpaceSettings {maxAngularAcceleration=10, maxAngularVelocity=10}
   }
   n60 = Anim Dynamics - Bone: dyn_backSkirt_01!(SimulationSpace:AnimPhysSimSpaceType Enum, ComponentPose=@n59, Alpha=1, GravityOverride:Vector="0,0,-980", ExternalForce="0,0,0") -> (Pose) {
    angularDampingOverride=1
    angularSpringConstant=20000
    bAngularSpring=true
    bOverrideAngularDamping=true
    bUseGravityOverride=true
    gravityScale=0.35
    simulationSpace=BoneRelative
    boundBone {boneName=dyn_backSkirt_01}
    componentAppliedLinearAccClamp=<500 500 500>
    componentLinearAccScale=<0.1 0.1 0.1>
    componentLinearVelScale=<0.2 0.2 0.2>
    gravityOverride {z=-980}
    physicsBodyDefinitions [{
      boundBone {boneName=dyn_backSkirt_01}
      boxExtents=<20 30 30>
      constraintSetup {angularLimitsMax=<0 0 0>, angularLimitsMin=<0 -3 0>, angularTarget=<1 0 0>}
      localJointOffset=<4 0 0>
     }]
    relativeSpaceBone {boneName=root}
    simSpaceSettings {maxAngularAcceleration=10, maxAngularVelocity=10}
   }
   n61 = Anim Dynamics - Bone: dyn_backSkirt_02!(SimulationSpace:AnimPhysSimSpaceType Enum, ComponentPose=@n60, Alpha=1, GravityOverride:Vector="0,0,-980", ExternalForce="0,0,0") -> (Pose) {
    angularDampingOverride=1
    angularSpringConstant=20000
    bAngularSpring=true
    bOverrideAngularDamping=true
    bUseGravityOverride=true
    gravityScale=0.35
    simulationSpace=BoneRelative
    boundBone {boneName=dyn_backSkirt_02}
    componentAppliedLinearAccClamp=<500 500 500>
    componentLinearAccScale=<0.1 0.1 0.1>
    componentLinearVelScale=<0.2 0.2 0.2>
    gravityOverride {z=-980}
    physicsBodyDefinitions [{
      boundBone {boneName=dyn_backSkirt_02}
      boxExtents=<20 30 30>
      constraintSetup {angularLimitsMax=<0 0 0>, angularLimitsMin=<0 -10 0>, angularTarget=<1 0 0>}
      localJointOffset=<4 0 0>
     }]
    relativeSpaceBone {boneName=root}
    simSpaceSettings {maxAngularAcceleration=10, maxAngularVelocity=10}
   }
   n62 = RigidBody(ComponentPose=@n61, OverrideWorldGravity="0,0,0") {
    alphaInputType=Bool
    bClampLinearTranslationLimitToRefPose=true
    bOverrideWorldGravity=true
    overridePhysicsAsset=&/M_MED_BlessFlan_Physics
    simulationSpace=BaseBoneSpace
    alphaBoolBlend {blendOption=Linear}
    baseBoneRef {boneName=pelvis}
    componentAppliedLinearAccClamp=<0 0 0>
    simSpaceSettings {
     maxAngularAcceleration=5
     maxAngularVelocity=5
     maxLinearAcceleration=5
     maxLinearVelocity=50
     velocityScaleZ=0.5
     worldAlpha=1
     externalLinearDragV=<2 2 2>
    }
   }
   Save cached pose 'DynPose'(Pose=Component To Local(ComponentPose=@n62)) {cacheName=DynPose}
   Use cached pose 'ControlPose'()
   Use cached pose 'ControlPose'()
   Get LeftMovement() -> (LeftMovement:double)
   Get BackwardMovement() -> (BackwardMovement:double)
   Get FrontEnd Idle() -> (FrontEnd Idle:bool)
  ]}
 EventGraph {nodes [
   n1 = Event BlueprintUpdateAnimation()
   > Set DeltaTimeX(DeltaTimeX:double=@n1.DeltaTimeX) -> (Output_Get:double)
   > n2 = Sequence!(execute) -> (then_0, then_1, then_2:Exec, then_3:Exec, then_4:Exec, then_5:Exec, then_6:Exec)
   > n3 = "Gravity Override State Function (Nativized)"!(
    execute
    CharacterPartAnimInstance:Custom Character Part Anim Instance Object Reference=Self-Reference()
    GravityOverrideParameterSet:Fort Gravity Override Parameter Set Structure=Get Gravity Override Params() -> (Gravity Override Params:Fort Gravity Override Parameter Set Structure)
    RigidBodyAnimNodeParameterSet:Fort Rigid Body Anim Node Parameter Set Structure=Get RBAN Params() -> (RBAN Params:Fort Rigid Body Anim Node Parameter Set Structure)
    DebugDraw:bool=false
    GameTimeInSeconds:float=GetGameTimeInSeconds()
    DebugJointName:Name=root
    StateLogicLayerName:Name=none
   ) -> (
    then
    OutGravityOverride:Vector
    OutCurrentState:Name=None
    OutGravityOverrideParameters:Fort Gravity Override Parameters Structure
    OutRigidBodyAnimNodeParameters:Fort Rigid Body Anim Node Parameters Structure
   )
   > Set Gravity Override(Gravity Override=@n3)
   > Set RBAN Settings Out(RBAN Settings Out=@n3.OutRigidBodyAnimNodeParameters)
   > Set Moving or Crouching(Moving or Crouching:bool=OR Boolean!(A=Get bIsMoving2D(), B=Get bIsCrouching(), C:bool=Get bIsJumping()) -> (ReturnValue)) -> (Output_Get:bool)
   n4 = Get isPistolEquipped()
   > Set TargetingNotPistol(TargetingNotPistol:bool=AND Boolean(A=NOT Boolean(A=@n4), B=Get bIsTargeting())) -> (Output_Get:bool)
   > Set IsCrouchMoving(IsCrouchMoving:bool=AND Boolean(A=Get bIsCrouching(), B=Get bIsMoving2D())) -> (Output_Get:bool)
   n5 = Get bIsMoving2D()
   n6 = Get bIsCrouching()
   n7 = Get bIsPlayingEmote()
   > Set IsStandingNoMovement(IsStandingNoMovement:bool=AND Boolean!(A=NOT Boolean(A=@n6), B=NOT Boolean(A=@n5), C:bool=NOT Boolean(A=@n7), D:bool=Get bIsOnGround() -> (bIsOnGround:bool)) -> (ReturnValue)) -> (Output_Get:bool)
   n8 = Get bIsCrouching()
   n9 = Get bIsPlayingEmote()
   > n10 = Set NotCrouching(NotCrouching:bool=AND Boolean!(A=NOT Boolean(A=@n8), B=Get bIsOnGround(), C=NOT Boolean(A=@n9)) -> (ReturnValue)) -> (Output_Get:bool)
   > Set NotCrouchingFloat(NotCrouchingFloat:double="To Float (Boolean)"(InBool=@n10)) -> (Output_Get:double)
   > Set StandingForward(StandingForward:double=float * float(A=Get NotCrouchingFloat() -> (NotCrouchingFloat:double), B=Get ForwardMovement()) {is=title K2Node_PromotableOperator}) -> (Output_Get:double)
   n11 = Get CurrentLOD()
   > Set Current LOD equal 1 ?(Current LOD equal 1 ?:bool="integer >= integer"(A=@n11, B=1)) -> (Output_Get:bool)
   n12 = PlayerMovement(execute=@n2.then_5, AnimInstance=Self-Reference(), XDotProductCoefficient=1, YDotProductCoefficient=1, ZDotProductCoefficient=1)
   > n13 = Set PlayerVector_LR(PlayerVector_LR:double=@n12) -> (Output_Get:double)
   > n14 = Set PlayerVector_FwdBwd(PlayerVector_FwdBwd:double=@n12.YMovement) -> (Output_Get:double)
   > Set ForwardMovement(ForwardMovement:double=@n14) -> (Output_Get:double)
   > Set BackwardMovement(BackwardMovement:double=float * float(A=@n14, B=-1)) -> (Output_Get:double)
   > Set RightMovement(RightMovement:double=@n13) -> (Output_Get:double)
   > Set LeftMovement(LeftMovement:double=float * float(A=@n13, B=-1)) -> (Output_Get:double)
   Set Is Not in FrontEnd(execute=@n2.then_6, Is Not in FrontEnd:bool=NOT Boolean(A=Get bIsInFrontEnd())) -> (Output_Get:bool)
   n15 = Get bIsPlayingEmote()
   n16 = Get bIsMoving2D()
   n17 = Get bIsSkyDiving()
   > Set FrontEnd Idle(FrontEnd Idle:bool=AND Boolean!(A=Get bIsInFrontEnd(), B=NOT Boolean(A=@n16), C=NOT Boolean(A=@n17), D:bool=NOT Boolean(A=@n15), E:bool=Get bRigidBodyEnabled()) -> (ReturnValue)) -> (Output_Get:bool)
   > Set Not Front End Idle(Not Front End Idle:bool=NOT Boolean(A=Get FrontEnd Idle())) -> (Output_Get:bool)
   > Set Not Front End Idle Float(Not Front End Idle Float:double="To Float (Boolean)"(InBool=Get Not Front End Idle())) -> (Output_Get:double)
   n18 = OR Boolean!(A=Get bIsPlayingEmote(), B=Get bIsSkyDiving()) -> (ReturnValue)
   Set is Emote Float(
    execute=@n2.then_1
    is Emote Float:double=FInterpTo_Constant!(
     Current:double=Get is Emote Float()
     Target:double="To Float (Boolean)"(InBool=@n18)
     DeltaTime:double=Get DeltaTimeX() -> (DeltaTimeX:double)
     InterpSpeed:double=4
    ) -> (ReturnValue:double)
   ) -> (Output_Get:double)
   n19 = GetAlphaFromReferenceTransformDelta(SkeletalMeshComponent=Get SkelMesh(), BoneName=spine_05, BaseName=head, InputRange=(X=0,Y=110), TwistAxis="0,0,1")
   Set Head Rot(execute=@n2.then_3, Head Rot:float=@n19) -> (Output_Get:float)
   n20 = "To Float (Boolean)"(InBool=Get isPistolEquipped())
   > Set Pistol UP(Pistol UP:float=float * float(A=@n19, B=@n20)) -> (Output_Get:float)
   n21 = Get SkelMesh()
   > Set Pistol Down(Pistol Down:float=float * float(
     A=@n20
     B=GetAlphaFromReferenceTransformDelta(SkeletalMeshComponent=@n21, BoneName=spine_05, BaseName=head, InputRange=(X=0,Y=-20), TwistAxis="0,0,1")
    )) -> (Output_Get:float)
   n22 = Cast To FortPlayerAnimInstance(execute=@n2.then_2, Object=GetAnimInstance(self=GetOwnerSkeletalMeshComponent()))
   > Set isPistolEquipped(isPistolEquipped:bool=false) -> (Output_Get:bool)
   > n23 = Switch on EFortWeaponCoreAnimation!(execute, Selection:EFortWeaponCoreAnimation Enum=Get WeaponCoreAnim(self=@n22)) -> (
    Melee:Exec
    Pistol:Exec
    Shotgun:Exec
    PaperBlueprint:Exec
    Rifle:Exec
    MeleeOneHand:Exec
    MachinePistol:Exec
    RocketLauncher:Exec
    GrenadeLauncher:Exec
    GoingCommando:Exec
    AssaultRifle:Exec
    TacticalShotgun:Exec
    SniperRifle:Exec
    TrapPlacement:Exec
    ShoulderLauncher:Exec
    AbilityDecoTool:Exec
    Crossbow:Exec
    C4:Exec
    RemoteControl:Exec
    DualWield:Exec
    AR_BullPup:Exec
    AR_ForwardGrip:Exec
    MedPackPaddles:Exec
    SMG_P90:Exec
    AR_DrumGun:Exec
    Consumable_Small:Exec
    Consumable_Large:Exec
    Balloon:Exec
    MountedTurret:Exec
    CreativeTool:Exec
    ExplosiveBow:Exec
    AshtonIndigo:Exec
    AshtonChicago:Exec
    MeleeDualWield:Exec
    Unarmed:Exec
   )
   n24 = Branch(execute=@n23.Pistol, Condition=Get bIsTargeting())
   Set isPistolEquipped(execute=[@n24.then, @n26.else], isPistolEquipped=true)
   n25 = Branch(execute=@n24.else, Condition=Get bIsRelaxedLevel1() -> (bIsRelaxedLevel1:bool))
   Set isPistolEquipped(execute=[@n25.then, @n26.then], isPistolEquipped=false)
   n26 = Branch(execute=@n25.else, Condition=Get bIsRelaxedLevel2() -> (bIsRelaxedLevel2:bool))
   n27 = CheckForBackbling()
   Set is Wearing Backbling(
    execute=@n2.then_4
    is Wearing Backbling:bool=AND Boolean!(A=@n27, B=NOT Boolean(A=@n27.bIsBackblingACape), C=NOT Boolean(A=@n27.bIsBackblingLong)) -> (ReturnValue)
   ) -> (Output_Get:bool)
   > Set is Wearing Cape(is Wearing Cape:bool=OR Boolean(A=@n27.bIsBackblingACape, B=@n27.bIsBackblingLong)) -> (Output_Get:bool)
   Event OnParachuteOpened()
   > Switch on EGliderType(Selection=Get PlayerGliderType() -> (PlayerGliderType=HangGlider))
   > Set Hang Gliding(Hang Gliding:bool=true) -> (Output_Get:bool)
   Event OnCharacterCustomizationLoaded()
   > n28 = Branch(Condition=Get bIsInFrontEnd())
   > Set Skirt Sim Space(Skirt Sim Space:AnimPhysSimSpaceType Enum=World) -> (Output_Get:AnimPhysSimSpaceType Enum)
   > Set Skirt Alpha(Skirt Alpha:float=0.1) -> (Output_Get:float)
   Set Skirt Sim Space(execute=@n28.else, Skirt Sim Space=BoneRelative)
   > Set Skirt Alpha(Skirt Alpha=1)
   Event OnParachuteClosed()
   > Set Hang Gliding(Hang Gliding=false)
   Event BlueprintInitializeAnimation()
   > Set SkelMesh(SkelMesh=GetOwnerSkeletalMeshComponent())
   "Absolute (Float)"(A=0)
   "Not Equal (Integer)"(A=@n11)
  ]}
]
variables [
 BackwardMovement {type=number}
 Current LOD equal 1 ? {type=boolean}
 DeltaTimeX {type=number}
 ForwardMovement {type=number}
 FrontEnd Idle {type=boolean}
 Gravity Override {type=Vector}
 Gravity Override Params {
  type=FortGravityOverrideParameterSet
  default=(Skydive_GoingUp=(GravityOverride=(WindFrequency=(X=1.500000,Y=7.000000,Z=8.000000),WindAmplitude=(X=200.000000,Y=-500.000000,Z=400.000000),GravityOverride=(X=0.000000,Y=-300.000000,Z=1500.000000),JointName="root")),Skydive_FloatingDown=(GravityOverride=(WindFrequency=(X=1.500000,Y=7.000000,Z=8.000000),WindAmplitude=(X=200.000000,Y=-500.000000,Z=400.000000),WindOffset=(X=0.000000,Y=-600.000000,Z=400.000000),GravityOverride=(X=0.000000,Y=-700.000000,Z=600.000000),JointName="root")),Skydive_DivingDown=(GravityOverride=(WindFrequency=(X=1.500000,Y=12.000000,Z=10.000000),WindAmplitude=(X=50.000000,Y=300.000000,Z=100.000000),GravityOverride=(X=-1500.000000,Y=0.000000,Z=0.000000),JointName="pelvis")),Skydive_Parachuting=(GravityOverride=(WindFrequency=(X=1.500000,Y=7.000000,Z=8.000000),WindAmplitude=(X=200.000000,Y=-500.000000,Z=400.000000),GravityOverride=(X=0.000000,Y=-800.000000,Z=-980.000000),JointName="root")),OnGround_Standing=(GravityOverride=(GravityOverride=(X=0.000000,Y=0.000000,Z=-980.000000),JointName="root")),OnGround_Moving=(GravityOverride=(GravityOverride=(X=0.000000,Y=0.000000,Z=-980.000000),JointName="root")),OnGround_Moving_HighRotationRate=(GravityOverride=(GravityOverride=(X=0.000000,Y=0.000000,Z=-980.000000),JointName="root")),OnGround_Crouch_Moving=(GravityOverride=(GravityOverride=(X=0.000000,Y=0.000000,Z=-980.000000),JointName="root")),Emote_or_Melee=(GravityOverride=(GravityOverride=(X=0.000000,Y=0.000000,Z=-980.000000),JointName="root")),Swimming=(GravityOverride=(GravityOverride=(X=-800.000000,Y=0.000000,Z=0.000000),JointName="pelvis")),Front_End=(GravityOverride=(GravityOverride=(X=0.000000,Y=0.000000,Z=-980.000000),JointName="root")),DBNO=(GravityOverride=(GravityOverride=(X=-980.000000,Y=400.000000,Z=0.000000),JointName="pelvis")),DBNO_Carried=(GravityOverride=(GravityOverride=(X=-980.000000,Y=500.000000,Z=-500.000000),JointName="pelvis")),OnGround_Falling=(GravityOverride=(GravityOverride=(X=0.000000,Y=0.000000,Z=-980.000000),JointName="root")),Ziplining=(GravityOverride=(GravityOverride=(X=0.000000,Y=0.000000,Z=-980.000000),JointName="root")),Skydive_Surfing=(GravityOverride=(WindFrequency=(X=1.500000,Y=7.000000,Z=8.000000),WindAmplitude=(X=200.000000,Y=-500.000000,Z=400.000000),GravityOverride=(X=500.000000,Y=-300.000000,Z=-980.000000),JointName="root")),Riding_Standing=(GravityOverride=(GravityOverride=(X=0.000000,Y=-350.000000,Z=-980.000000),JointName="root")),Riding_Moving=(GravityOverride=(GravityOverride=(X=0.000000,Y=-450.000000,Z=-450.000000),JointName="root")),Riding_Moving_Targeting=(GravityOverride=(GravityOverride=(X=0.000000,Y=-200.000000,Z=-980.000000),JointName="root")),Tactical_Sprint=(GravityOverride=(GravityOverride=(X=0.000000,Y=-450.000000,Z=-500.000000),JointName="root")),Rail_Grinding=(GravityOverride=(GravityOverride=(X=0.000000,Y=0.000000,Z=-980.000000),JointName="root")),)
 }
 Hang Gliding {type=boolean}
 Head Rot {type=number}
 is Emote Float {type=number}
 Is Not in FrontEnd {type=boolean}
 is Wearing Backbling {type=boolean}
 is Wearing Cape {type=boolean}
 IsCrouchMoving {type=boolean}
 isPistolEquipped {type=boolean}
 IsStandingNoMovement {type=boolean}
 LeftMovement {type=number}
 Moving or Crouching {type=boolean}
 Not Front End Idle {type=boolean}
 Not Front End Idle Float {type=number}
 NotCrouching {type=boolean}
 NotCrouchingFloat {type=number}
 Pistol Down {type=number}
 Pistol UP {type=number}
 PlayerVector_FwdBwd {type=number}
 PlayerVector_LR {type=number}
 RBAN Params {
  type=FortRigidBodyAnimNodeParameterSet
  default=(Skydive_GoingUp=(RigidBodyAnimNode=(SimSpaceSettings=(WorldAlpha=1.000000,MaxLinearVelocity=500.000000,MaxAngularVelocity=3.000000,MaxLinearAcceleration=800.000000,MaxAngularAcceleration=1.000000,ExternalLinearDragV=(X=1.000000,Y=1.000000,Z=1.000000)))),Skydive_FloatingDown=(RigidBodyAnimNode=(SimSpaceSettings=(WorldAlpha=1.000000,MaxLinearVelocity=500.000000,MaxAngularVelocity=3.000000,MaxLinearAcceleration=800.000000,MaxAngularAcceleration=1.000000,ExternalLinearDragV=(X=1.000000,Y=1.000000,Z=1.000000)))),Skydive_DivingDown=(RigidBodyAnimNode=(SimSpaceSettings=(WorldAlpha=1.000000,MaxLinearVelocity=500.000000,MaxAngularVelocity=3.000000,MaxLinearAcceleration=800.000000,MaxAngularAcceleration=1.000000,ExternalLinearDragV=(X=1.000000,Y=1.000000,Z=1.000000)))),Skydive_Parachuting=(RigidBodyAnimNode=(SimSpaceSettings=(WorldAlpha=1.000000,MaxLinearVelocity=500.000000,MaxAngularVelocity=3.000000,MaxLinearAcceleration=800.000000,MaxAngularAcceleration=1.000000,ExternalLinearDragV=(X=1.000000,Y=1.000000,Z=1.000000)))),OnGround_Standing=(RigidBodyAnimNode=(SimSpaceSettings=(WorldAlpha=1.000000,MaxLinearVelocity=500.000000,MaxAngularVelocity=3.000000,MaxLinearAcceleration=800.000000,MaxAngularAcceleration=1.000000,ExternalLinearDragV=(X=1.000000,Y=1.000000,Z=1.000000)))),OnGround_Moving=(RigidBodyAnimNode=(SimSpaceSettings=(WorldAlpha=1.000000,MaxLinearVelocity=500.000000,MaxAngularVelocity=3.000000,MaxLinearAcceleration=800.000000,MaxAngularAcceleration=1.000000,ExternalLinearDragV=(X=1.000000,Y=1.000000,Z=1.000000)))),OnGround_Moving_HighRotationRate=(RigidBodyAnimNode=(SimSpaceSettings=(WorldAlpha=1.000000,MaxLinearVelocity=500.000000,MaxAngularVelocity=3.000000,MaxLinearAcceleration=800.000000,MaxAngularAcceleration=1.000000,ExternalLinearDragV=(X=1.000000,Y=1.000000,Z=1.000000)))),OnGround_Crouch_Moving=(RigidBodyAnimNode=(SimSpaceSettings=(WorldAlpha=1.000000,MaxLinearVelocity=500.000000,MaxAngularVelocity=3.000000,MaxLinearAcceleration=800.000000,MaxAngularAcceleration=1.000000,ExternalLinearDragV=(X=1.000000,Y=1.000000,Z=1.000000)))),Emote_or_Melee=(RigidBodyAnimNode=(SimSpaceSettings=(WorldAlpha=1.000000,MaxLinearVelocity=500.000000,MaxAngularVelocity=3.000000,MaxLinearAcceleration=1000.000000,MaxAngularAcceleration=1.000000,ExternalLinearDragV=(X=1.000000,Y=1.000000,Z=1.000000)))),Swimming=(RigidBodyAnimNode=(SimSpaceSettings=(WorldAlpha=1.000000,MaxLinearVelocity=500.000000,MaxAngularVelocity=3.000000,MaxLinearAcceleration=800.000000,MaxAngularAcceleration=1.000000,ExternalLinearDragV=(X=1.000000,Y=1.000000,Z=1.000000)))),Front_End=(RigidBodyAnimNode=(SimSpaceSettings=(WorldAlpha=1.000000,MaxLinearVelocity=500.000000,MaxAngularVelocity=3.000000,MaxLinearAcceleration=800.000000,MaxAngularAcceleration=1.000000,ExternalLinearDragV=(X=1.000000,Y=1.000000,Z=1.000000)))),DBNO=(RigidBodyAnimNode=(SimSpaceSettings=(WorldAlpha=1.000000,MaxLinearVelocity=500.000000,MaxAngularVelocity=3.000000,MaxLinearAcceleration=800.000000,MaxAngularAcceleration=1.000000,ExternalLinearDragV=(X=1.000000,Y=1.000000,Z=1.000000)))),DBNO_Carried=(RigidBodyAnimNode=(SimSpaceSettings=(WorldAlpha=1.000000,MaxLinearVelocity=500.000000,MaxAngularVelocity=3.000000,MaxLinearAcceleration=800.000000,MaxAngularAcceleration=1.000000,ExternalLinearDragV=(X=1.000000,Y=1.000000,Z=1.000000)))),OnGround_Falling=(RigidBodyAnimNode=(SimSpaceSettings=(WorldAlpha=1.000000,MaxLinearVelocity=500.000000,MaxAngularVelocity=3.000000,MaxLinearAcceleration=800.000000,MaxAngularAcceleration=1.000000,ExternalLinearDragV=(X=1.000000,Y=1.000000,Z=1.000000)))),Ziplining=(RigidBodyAnimNode=(SimSpaceSettings=(WorldAlpha=1.000000,MaxLinearVelocity=500.000000,MaxAngularVelocity=3.000000,MaxLinearAcceleration=800.000000,MaxAngularAcceleration=1.000000,ExternalLinearDragV=(X=1.000000,Y=1.000000,Z=1.000000)))),Skydive_Surfing=(RigidBodyAnimNode=(SimSpaceSettings=(WorldAlpha=1.000000,MaxLinearVelocity=500.000000,MaxAngularVelocity=3.000000,MaxLinearAcceleration=800.000000,MaxAngularAcceleration=1.000000,ExternalLinearDragV=(X=1.000000,Y=1.000000,Z=1.000000)))),Riding_Standing=(RigidBodyAnimNode=(SimSpaceSettings=(WorldAlpha=1.000000,MaxLinearVelocity=500.000000,MaxAngularVelocity=3.000000,MaxLinearAcceleration=800.000000,MaxAngularAcceleration=1.000000,ExternalLinearDragV=(X=1.000000,Y=1.000000,Z=1.000000)))),Riding_Moving=(RigidBodyAnimNode=(SimSpaceSettings=(WorldAlpha=1.000000,MaxLinearVelocity=500.000000,MaxAngularVelocity=3.000000,MaxLinearAcceleration=800.000000,MaxAngularAcceleration=1.000000,ExternalLinearDragV=(X=1.000000,Y=1.000000,Z=1.000000)))),Riding_Moving_Targeting=(RigidBodyAnimNode=(SimSpaceSettings=(WorldAlpha=1.000000,MaxLinearVelocity=500.000000,MaxAngularVelocity=3.000000,MaxLinearAcceleration=1000.000000,MaxAngularAcceleration=1.000000,ExternalLinearDragV=(X=1.000000,Y=1.000000,Z=1.000000)))),Tactical_Sprint=(RigidBodyAnimNode=(SimSpaceSettings=(WorldAlpha=1.000000,MaxLinearVelocity=500.000000,MaxAngularVelocity=3.000000,MaxLinearAcceleration=800.000000,MaxAngularAcceleration=1.000000,ExternalLinearDragV=(X=1.000000,Y=1.000000,Z=1.000000)))),Rail_Grinding=(RigidBodyAnimNode=(SimSpaceSettings=(WorldAlpha=1.000000,MaxLinearVelocity=500.000000,MaxAngularVelocity=3.000000,MaxLinearAcceleration=800.000000,MaxAngularAcceleration=1.000000,ExternalLinearDragV=(X=1.000000,Y=1.000000,Z=1.000000)))),)
 }
 RBAN Settings Out {type=FortRigidBodyAnimNodeParameters}
 RightMovement {type=number}
 SkelMesh {type=SkeletalMeshComponent}
 Skirt Alpha {type=number}
 Skirt Sim Space {type=AnimPhysSimSpaceType}
 StandingForward {type=number}
 TargetingNotPistol {type=boolean}
]
