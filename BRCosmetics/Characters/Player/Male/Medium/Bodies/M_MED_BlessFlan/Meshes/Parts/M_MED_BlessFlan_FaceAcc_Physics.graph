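# Physics asset for the BlessFlan male medium face accessory, setting up the dangling headband tails for physics motion against the upper body.
# Two capsule chains hang from the headband bones under the head: four left segments with tight angular limits and three right segments that loosen past the first.
# Capsules on spine, neck, head, clavicles, left upper arm and right pauldron form body volumes with free-rotating joints.
# The right upper arm body has no shape.
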PhysicsAsset /BRCosmetics/Characters/Player/Male/Medium/Bodies/M_MED_BlessFlan/Meshes/Parts/M_MED_BlessFlan_FaceAcc_Physics
bodies {
 base_headband_l [{shapeType=Capsule, length=0.505, radius=0.505, shapeName=base_headband_l_capsule, rotation=<rot 0 -90 0>}]
 base_headband_r [{shapeType=Capsule, length=0.505, radius=0.505, shapeName=base_headband_r_capsule, rotation=<rot 0 -90 0>}]
 clavicle_l [{shapeType=Capsule, length=4, radius=9, shapeName=clavicle_l_capsule, center=<7 0 0>, rotation=<rot 90 0 0>}]
 clavicle_r [{shapeType=Capsule, length=4, radius=9, shapeName=clavicle_r_capsule, center=<-7 0 0>, rotation=<rot 90 0 0>}]
 dyn_headband_l_01 [{shapeType=Capsule, length=4, radius=1, shapeName=dyn_headband_l_01_capsule, center=<3.867 0 0>, rotation=<rot 90 0 0>}]
 dyn_headband_l_02 [{shapeType=Capsule, length=4.5, radius=1, shapeName=dyn_headband_l_02_capsule, center=<3.6185 0 0>, rotation=<rot 90 0 0>}]
 dyn_headband_l_03 [{shapeType=Capsule, length=4, radius=1, shapeName=dyn_headband_l_03_capsule, center=<3.0306 0 0>, rotation=<rot 90 0 0>}]
 dyn_headband_l_04 [{shapeType=Capsule, length=3.5, radius=1, shapeName=dyn_headband_l_04_capsule, center=<2.7851 0 0>, rotation=<rot 90 0 0>}]
 dyn_headband_r_01 [{shapeType=Capsule, length=4.5, radius=1, shapeName=dyn_headband_r_01_capsule, center=<3.3818 0 0>, rotation=<rot 90 0 0>}]
 dyn_headband_r_02 [{shapeType=Capsule, length=6, radius=1, shapeName=dyn_headband_r_02_capsule, center=<4.09 0 0>, rotation=<rot 90 0 0>}]
 dyn_headband_r_03 [{shapeType=Capsule, length=9.6836, radius=1.1887, shapeName=dyn_headband_r_03_capsule, center=<5.9075 0 0>, rotation=<rot 90 0 0>}]
 head [{shapeType=Capsule, length=11.8505, radius=8.8174, shapeName=head_capsule, center=<2.6172 1.1131 0>, rotation=<rot 90 4.9991 0>}]
 headband_MSTR [{shapeType=Capsule, length=0.505, radius=0.505, shapeName=headband_MSTR_capsule, rotation=<rot 0 -90 0>}]
 neck_01 [{shapeType=Capsule, length=2, radius=7, shapeName=neck_01_capsule, rotation=<rot -90 -0.0002 0>}]
 neck_02 [{shapeType=Capsule, length=2, radius=7, shapeName=neck_02_capsule, rotation=<rot -90 -0.0002 0>}]
 pauldron_r [{shapeType=Capsule, length=1, radius=11, shapeName=pauldron_r_capsule, center=<2.3083 1.4278 2.4181>, rotation=<rot 90 0 0>}]
 spine_05 [{shapeType=Capsule, length=12.5, radius=12, shapeName=spine_05_capsule, center=<0.5484 0.2251 0>}]
 upperarm_l [{shapeType=Capsule, length=9.5, radius=5.5, shapeName=upperarm_l_capsule, center=<5 0 0>, rotation=<rot 90 0 0>}]
 upperarm_r []
}
constraints [
 {
  bone1Name=neck_01
  bone2Name=spine_05
  swing1LimitDegrees=45
  swing1Motion=Free
  swing2LimitDegrees=45
  swing2Motion=Free
  twistLimitDegrees=45
  twistMotion=Free
 }
 {
  bone1Name=neck_02
  bone2Name=neck_01
  swing1LimitDegrees=45
  swing1Motion=Free
  swing2LimitDegrees=45
  swing2Motion=Free
  twistLimitDegrees=45
  twistMotion=Free
 }
 {
  bone1Name=head
  bone2Name=neck_02
  swing1LimitDegrees=45
  swing1Motion=Free
  swing2LimitDegrees=45
  swing2Motion=Free
  twistLimitDegrees=45
  twistMotion=Free
 }
 {
  bone1Name=headband_MSTR
  bone2Name=head
  swing1LimitDegrees=45
  swing1Motion=Free
  swing2LimitDegrees=45
  swing2Motion=Free
  twistLimitDegrees=45
  twistMotion=Free
 }
 {
  bone1Name=base_headband_l
  bone2Name=headband_MSTR
  swing1LimitDegrees=45
  swing1Motion=Free
  swing2LimitDegrees=45
  swing2Motion=Free
  twistLimitDegrees=45
  twistMotion=Free
 }
 {bone1Name=dyn_headband_l_01, bone2Name=base_headband_l, swing1LimitDegrees=15, swing2LimitDegrees=10, twistLimitDegrees=5}
 {bone1Name=dyn_headband_l_02, bone2Name=dyn_headband_l_01, swing1LimitDegrees=15, swing2LimitDegrees=10, twistLimitDegrees=5}
 {bone1Name=dyn_headband_l_03, bone2Name=dyn_headband_l_02, swing1LimitDegrees=15, swing2LimitDegrees=10, twistLimitDegrees=5}
 {bone1Name=dyn_headband_l_04, bone2Name=dyn_headband_l_03, swing1LimitDegrees=15, swing2LimitDegrees=10, twistLimitDegrees=5}
 {
  bone1Name=base_headband_r
  bone2Name=headband_MSTR
  swing1LimitDegrees=45
  swing1Motion=Free
  swing2LimitDegrees=45
  swing2Motion=Free
  twistLimitDegrees=45
  twistMotion=Free
 }
 {bone1Name=dyn_headband_r_01, bone2Name=base_headband_r, swing1LimitDegrees=15, swing2LimitDegrees=20, twistLimitDegrees=5}
 {bone1Name=dyn_headband_r_02, bone2Name=dyn_headband_r_01, swing1LimitDegrees=30, swing2LimitDegrees=30, twistLimitDegrees=5}
 {bone1Name=dyn_headband_r_03, bone2Name=dyn_headband_r_02, swing1LimitDegrees=30, swing2LimitDegrees=30, twistLimitDegrees=5}
 {
  bone1Name=clavicle_l
  bone2Name=spine_05
  swing1LimitDegrees=45
  swing1Motion=Free
  swing2LimitDegrees=45
  swing2Motion=Free
  twistLimitDegrees=45
  twistMotion=Free
 }
 {
  bone1Name=upperarm_l
  bone2Name=clavicle_l
  swing1LimitDegrees=45
  swing1Motion=Free
  swing2LimitDegrees=45
  swing2Motion=Free
  twistLimitDegrees=45
  twistMotion=Free
 }
 {
  bone1Name=clavicle_r
  bone2Name=spine_05
  swing1LimitDegrees=45
  swing1Motion=Free
  swing2LimitDegrees=45
  swing2Motion=Free
  twistLimitDegrees=45
  twistMotion=Free
 }
 {
  bone1Name=upperarm_r
  bone2Name=clavicle_r
  swing1LimitDegrees=45
  swing1Motion=Free
  swing2LimitDegrees=45
  swing2Motion=Free
  twistLimitDegrees=45
  twistMotion=Free
 }
 {
  bone1Name=pauldron_r
  bone2Name=clavicle_r
  swing1LimitDegrees=45
  swing1Motion=Free
  swing2LimitDegrees=45
  swing2Motion=Free
  twistLimitDegrees=45
  twistMotion=Free
 }
]
